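# Control Rig for the F_MED_ForestJeanGrowl face accessory that adjusts the character's ponytail bones to follow head and body motion.
# UseSplineHair, off by default, selects HairSpline over Hair Normal, which applies a head-driven pose-reader corrective only when RbanOff is set.
# HairSpline combines aim and weighted parent constraints, pose-reader offsets and SplineIK.
# SplineIK fits bones to a control-point spline, orienting them toward a second, offset spline.
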ControlRigBlueprint /BRCosmetics/Characters/Player/Female/Medium/Bodies/F_MED_ForestJeanGrowl/Meshes/Parts/F_MED_ForestJeanGrowl_FaceAcc_CtrlRig
graphs [
 ContainedGraph {nodes [
   BranchNode_RigVMUnitNode = &RigVMFunction_ControlFlowBranch(ExecuteContext=(), Condition=RigVMVariableNode() {variable=debug}.Value)
   /Script/ControlRigSpline.RigUnit_DrawControlRigSpline(
    Color=(R=1,G=0,B=0,A=1)
    Detail=16
    Thickness=0.75
    ExecutePin=@BranchNode_RigVMUnitNode.True
    Spline=RigVMVariableNode() {variable=template_spline}.Value
   ) {is=struct RigVMUnitNode}
   n1 = RigVMFunctionEntryNode()
   n2 = RigVMVariableNode() {variable=Controls}
   BranchNode_4_RigVMUnitNode = &RigVMFunction_ControlFlowBranch(
    ExecuteContext=@n1.ExecuteContext
    Condition=/Script/RigVM.RigVMFunction_MathIntLess(B=4, A=RigVMDispatchNode(Array=@n2.Value) {template="DISPATCH_RigVMDispatch_ArrayGetNum(in Array,out Num)"}.Num) {is=struct RigVMUnitNode, template="Less::Execute(in A,in B,out Result)"}.Result
   )
   RigVMFunctionReturnNode(ExecuteContext=@BranchNode_4_RigVMUnitNode.Completed, Output Spline=RigVMVariableNode() {variable=local_spline}.Value)
   RigUnit_SequenceExecution_1 = RigVMAggregateNode(ExecuteContext=@BranchNode_4_RigVMUnitNode.False)
   n3 = /Script/ControlRig.RigUnit_GetTransformItemArray(Space=GlobalSpace, bInitial=False, Items=RigVMVariableNode() {variable=Controls}.Value) {is=struct RigVMUnitNode}
   RigVMFunction_MathTransformArrayToSRT = /Script/RigVM.RigVMFunction_MathTransformArrayToSRT(Transforms=@n3.Transforms) {is=struct RigVMUnitNode}
   RigVMVariableNode(
    Value=/Script/ControlRigSpline.RigUnit_ControlRigSplineFromPoints(
     bClosed=False
     Points=@RigVMFunction_MathTransformArrayToSRT.Translations
     SplineMode=RigVMVariableNode() {variable=Spline Mode}.Value
     SamplesPerSegment=RigVMVariableNode() {variable=Samples Per Segment}.Value
    ) {is=struct RigVMUnitNode}.Spline
    ExecuteContext=@RigUnit_SequenceExecution_1.A
   ) {variable=template_spline}
   > RigVMVariableNode(Value=@n3.Transforms) {variable=template_transforms}
   > RigUnit_SequenceExecution = RigVMAggregateNode()
   n4 = RigVMVariableNode() {variable=template_transforms}
   RigVMDispatchNode(
    Num=RigVMDispatchNode(Array=@n4.Value) {template="DISPATCH_RigVMDispatch_ArrayGetNum(in Array,out Num)"}.Num
    ExecuteContext=@RigUnit_SequenceExecution.A
    Array=RigVMVariableNode() {variable=offset_positions}.Value
   ) {template="DISPATCH_RigVMDispatch_ArraySetNum(io Array,in Num)"}
   > n5 = RigVMDispatchNode(Array=@n4.Value) {template="DISPATCH_RigVMDispatch_ArrayIterator(in Array,out Element,out Index,out Count,out Ratio)"}
   > RigVMDispatchNode(
    Index=@n5.Index
    Element=/Script/RigVM.RigVMFunction_MathTransformTransformVector(Transform=@n5.Element, Location=RigVMVariableNode() {variable=Secondary Spline Direction}.Value) {is=struct RigVMUnitNode}.Result
    Array=RigVMVariableNode() {variable=offset_positions}.Value
   ) {template="DISPATCH_RigVMDispatch_ArraySetAtIndex(io Array,in Index,in Element)"}
   RigUnit_SequenceExecution_4 = RigVMAggregateNode(ExecuteContext=@RigUnit_SequenceExecution.C)
   Branch_RigVMUnitNode = &RigVMFunction_ControlFlowBranch(ExecuteContext=@RigUnit_SequenceExecution_4.A, Condition=RigVMVariableNode() {variable=Stretch}.Value)
   n6 = RigVMVariableNode() {variable=template_spline}
   n7 = RigVMVariableNode() {variable=Primary Axis}
   n8 = RigVMVariableNode() {variable=Bones}
   /Script/ControlRigSpline.RigUnit_FitChainToSplineCurveItemArray(
    Alignment=Front
    DebugSettings=(Scale=1,CurveColor=(R=1,G=1,B=0,A=1),SegmentsColor=(R=1,G=0,B=0,A=1),WorldOffset=(Rotation=(X=0,Y=0,Z=0,W=1),Translation=(X=0,Y=0,Z=0),Scale3D=(X=1,Y=1,Z=1)))
    Maximum=1
    PoleVectorPosition=(X=0,Y=0,Z=0)
    Rotations=()
    SamplingPrecision=12
    SecondaryAxis=(X=0,Y=0,Z=0)
    Weight=1
    bPropagateToChildren=True
    ExecutePin=@Branch_RigVMUnitNode.False
    Spline=@n6.Value
    PrimaryAxis=@n7.Value
    Items=@n8.Value
   ) {is=struct RigVMUnitNode}
   &RigUnit_FitChainToSplineCurveItemArray(
    Alignment=Stretched
    DebugSettings=(Scale=1,CurveColor=(R=1,G=1,B=0,A=1),SegmentsColor=(R=1,G=0,B=0,A=1),WorldOffset=(Rotation=(X=0,Y=0,Z=0,W=1),Translation=(X=0,Y=0,Z=0),Scale3D=(X=1,Y=1,Z=1)))
    Maximum=1
    PoleVectorPosition=(X=0,Y=0,Z=0)
    Rotations=()
    SamplingPrecision=12
    SecondaryAxis=(X=0,Y=0,Z=0)
    Weight=1
    bPropagateToChildren=True
    ExecutePin=@Branch_RigVMUnitNode.True
    Spline=@n6.Value
    PrimaryAxis=@n7.Value
    Items=@n8.Value
   )
   RigUnit_SequenceExecution_2 = RigVMAggregateNode(ExecuteContext=@RigUnit_SequenceExecution_4.B)
   n9 = RigVMVariableNode(
    Value=&RigUnit_ControlRigSplineFromPoints(
     Stretch=0
     bClosed=False
     SplineMode=RigVMVariableNode() {variable=Spline Mode}.Value
     SamplesPerSegment=RigVMVariableNode() {variable=Samples Per Segment}.Value
     Points=RigVMVariableNode() {variable=offset_positions}.Value
    ).Spline
    ExecuteContext=@RigUnit_SequenceExecution_2.A
   ) {variable=local_spline}
   BranchNode_2_RigVMUnitNode = &RigVMFunction_ControlFlowBranch(ExecuteContext=@n9.ExecuteContext, Condition=RigVMVariableNode() {variable=debug}.Value)
   &RigUnit_DrawControlRigSpline(
    Color=(R=1,G=0,B=0,A=1)
    Detail=16
    Thickness=1
    ExecutePin=@BranchNode_2_RigVMUnitNode.True
    Spline=RigVMVariableNode() {variable=local_spline}.Value
   )
   n10 = RigVMVariableNode() {variable=Bones}
   DISPATCH_RigVMDispatch_ArrayGetNum = RigVMDispatchNode(Array=@n10.Value) {template="DISPATCH_RigVMDispatch_ArrayGetNum(in Array,out Num)"}
   DISPATCH_RigVMDispatch_ArraySetNum = RigVMDispatchNode(
    Num=@DISPATCH_RigVMDispatch_ArrayGetNum.Num
    ExecuteContext=@RigUnit_SequenceExecution_2.B
    Array=RigVMVariableNode() {variable=aim_transforms}.Value
   ) {template="DISPATCH_RigVMDispatch_ArraySetNum(io Array,in Num)"}
   > n11 = RigVMDispatchNode(Array=@n10.Value) {template="DISPATCH_RigVMDispatch_ArrayIterator(in Array,out Element,out Index,out Count,out Ratio)"}
   n12 = &RigUnit_GetTransform(Item=@n11.Element)
   n13 = RigVMVariableNode() {variable=local_spline}
   > DISPATCH_RigVMDispatch_ArraySetAtIndex = RigVMDispatchNode(
    Index=@n11.Index
    Element=/Script/ControlRigSpline.RigUnit_PositionFromControlRigSpline(
     U=/Script/ControlRigSpline.RigUnit_ClosestParameterFromControlRigSpline(Position=@n12.Transform.Translation, Spline=@n13.Value) {is=struct RigVMUnitNode}.U
     Spline=@n13.Value
    ) {is=struct RigVMUnitNode}.Position
    Array=RigVMVariableNode() {variable=aim_transforms}.Value
   ) {template="DISPATCH_RigVMDispatch_ArraySetAtIndex(io Array,in Index,in Element)"}
   n14 = RigVMDispatchNode(ExecuteContext=@RigUnit_SequenceExecution.D, Array=RigVMVariableNode() {variable=Bones}.Value) {template="DISPATCH_RigVMDispatch_ArrayIterator(in Array,out Element,out Index,out Count,out Ratio)"}
   DISPATCH_RigVMDispatch_ArrayGetAtIndex = RigVMDispatchNode(Index=@n14.Index, Array=RigVMVariableNode() {variable=aim_transforms}.Value) {template="DISPATCH_RigVMDispatch_ArrayGetAtIndex(in Array,in Index,out Element)"}
   n15 = /Script/ControlRig.RigUnit_AimBoneMath(
    DebugSettings=(Scale=10,WorldOffset=(Rotation=(X=0,Y=0,Z=0,W=1),Translation=(X=0,Y=0,Z=0),Scale3D=(X=1,Y=1,Z=1)))
    Primary=(Weight=1,Axis=(X=0,Y=1,Z=0),Target=(X=1,Y=0,Z=0),Kind=Location,Space=(Type=Bone))
    Secondary=(Weight=0,Axis=(X=0,Y=0,Z=0),Target=(X=0,Y=0,Z=0),Kind=Location,Space=(Type=Bone))
    Weight=1
    Primary.Target=@DISPATCH_RigVMDispatch_ArrayGetAtIndex.Element
    InputTransform=&RigUnit_GetTransform(Item=@n14.Element).Transform
    Primary.Axis=RigVMVariableNode() {variable=Up Axis}.Value
   ) {is=struct RigVMUnitNode}
   MathIntEquals = RigVMDispatchNode(
    A=@n14.Index
    B=/Script/RigVM.RigVMFunction_MathIntSub(B=1, A=@n14.Count) {is=struct RigVMUnitNode, template="Subtract::Execute(in A,in B,out Result)"}.Result
   ) {template="DISPATCH_RigVMDispatch_CoreEquals(in A,in B,out Result)"}
   > n16 = &RigUnit_SetTransform(Weight=1, bInitial=False, Value=@n15.Result, Item=@n14.Element, bPropagateToChildren=@MathIntEquals.Result)
   BranchNode_3_RigVMUnitNode = &RigVMFunction_ControlFlowBranch(ExecuteContext=@n16.ExecutePin, Condition=RigVMVariableNode() {variable=debug}.Value)
   /Script/ControlRig.RigUnit_DebugTransformMutableItemSpace(
    Color=(R=1,G=0,B=0,A=1)
    DebugDrawSettings=(DepthPriority=SDPG_Foreground,Lifetime=-1)
    Mode=Axes
    Scale=10
    Space=(Type=Bone)
    Thickness=0.15
    WorldOffset=(Rotation=(X=0,Y=0,Z=0,W=1),Translation=(X=0,Y=0,Z=0),Scale3D=(X=1,Y=1,Z=1))
    bEnabled=True
    Transform=@n15.Result
    ExecutePin=@BranchNode_3_RigVMUnitNode.True
   ) {is=struct RigVMUnitNode}
   Branch_1_RigVMUnitNode = &RigVMFunction_ControlFlowBranch(ExecuteContext=@RigUnit_SequenceExecution_1.B, Condition=RigVMVariableNode() {variable=Stretch}.Value)
   n17 = RigVMDispatchNode(Index=-1, Array=RigVMVariableNode() {variable=Controls}.Value) {template="DISPATCH_RigVMDispatch_ArrayGetAtIndex(in Array,in Index,out Element)"}
   n18 = RigVMDispatchNode(Index=-1, Array=RigVMVariableNode() {variable=Bones}.Value) {template="DISPATCH_RigVMDispatch_ArrayGetAtIndex(in Array,in Index,out Element)"}
   &RigUnit_ParentConstraint(
    AdvancedSettings=(RotationOrderForFilter=XZY)
    Filter=(TranslationFilter=(bX=false,bY=false,bZ=false),RotationFilter=(bX=True,bY=True,bZ=True),ScaleFilter=(bX=True,bY=True,bZ=True))
    Parents=((Item=(Type=Bone),Weight=1))
    Weight=1
    Parents.0.Item=@n17.Element
    Child=@n18.Element
    ExecutePin=@Branch_1_RigVMUnitNode.False
   )
   &RigUnit_ParentConstraint(
    AdvancedSettings=(RotationOrderForFilter=XZY)
    Parents=((Item=(Type=Bone),Weight=1))
    Weight=1
    Parents.0.Item=@n17.Element
    Child=@n18.Element
    ExecutePin=@Branch_1_RigVMUnitNode.True
   )
  ]}
 ContainedGraph {nodes [
   n1 = RigVMFunctionEntryNode()
   RigUnit_SequenceExecution_2 = &RigVMFunction_Sequence(ExecuteContext=@n1.ExecuteContext)
   RigUnit_SequenceExecution_2_1 = &RigVMFunction_Sequence(ExecuteContext=@RigUnit_SequenceExecution_2.B)
   RigUnit_SequenceExecution_2_2 = &RigVMFunction_Sequence(ExecuteContext=@RigUnit_SequenceExecution_2_1.B)
   RigVMFunctionReturnNode(A=@RigUnit_SequenceExecution_2.A, B=@RigUnit_SequenceExecution_2_1.A, C=@RigUnit_SequenceExecution_2_2.A, D=@RigUnit_SequenceExecution_2_2.B)
  ]}
 ContainedGraph {nodes [
   n1 = RigVMFunctionEntryNode()
   RigUnit_SequenceExecution_4 = &RigVMFunction_Sequence(ExecuteContext=@n1.ExecuteContext)
   RigUnit_SequenceExecution_4_1 = &RigVMFunction_Sequence(ExecuteContext=@RigUnit_SequenceExecution_4.B)
   RigUnit_SequenceExecution_4_2 = &RigVMFunction_Sequence(ExecuteContext=@RigUnit_SequenceExecution_4_1.B)
   RigVMFunctionReturnNode(A=@RigUnit_SequenceExecution_4.A, B=@RigUnit_SequenceExecution_4_1.A, C=@RigUnit_SequenceExecution_4_2.A, D=@RigUnit_SequenceExecution_4_2.B)
  ]}
 ContainedGraph {nodes [
   n1 = RigVMFunctionEntryNode()
   RigUnit_SequenceExecution = &RigVMFunction_Sequence(ExecuteContext=@n1.ExecuteContext)
   RigUnit_SequenceExecution_1 = &RigVMFunction_Sequence(ExecuteContext=@RigUnit_SequenceExecution.B)
   RigUnit_SequenceExecution_2 = &RigVMFunction_Sequence(ExecuteContext=@RigUnit_SequenceExecution_1.B)
   RigVMFunctionReturnNode(A=@RigUnit_SequenceExecution.A, B=@RigUnit_SequenceExecution_1.A, C=@RigUnit_SequenceExecution_2.A, D=@RigUnit_SequenceExecution_2.B)
  ]}
 ContainedGraph {nodes [
   n1 = RigVMFunctionEntryNode()
   RigUnit_SequenceExecution_1 = &RigVMFunction_Sequence(ExecuteContext=@n1.ExecuteContext)
   RigUnit_SequenceExecution_1_1 = &RigVMFunction_Sequence(ExecuteContext=@RigUnit_SequenceExecution_1.B)
   RigUnit_SequenceExecution_1_2 = &RigVMFunction_Sequence(ExecuteContext=@RigUnit_SequenceExecution_1_1.B)
   RigVMFunctionReturnNode(A=@RigUnit_SequenceExecution_1.A, B=@RigUnit_SequenceExecution_1_1.A, C=@RigUnit_SequenceExecution_1_2.A, D=@RigUnit_SequenceExecution_1_2.B)
  ]}
 ContainedGraph {nodes [
   n1 = RigVMFunctionEntryNode()
   n2 = /Script/RigVM.RigVMFunction_MathTransformToEulerTransform(Value=@n1.transform) {is=struct RigVMUnitNode, template="Cast::Execute(in Value,out Result)"}
   RigVMFunction_MathQuaternionFromRotatorV2 = /Script/RigVM.RigVMFunction_MathQuaternionFromRotatorV2(Value=@n2.Result.Rotation) {is=struct RigVMUnitNode, template="Cast::Execute(in Value,out Result)"}
   > RigVMFunctionReturnNode(
    mirrorTransform=(Rotation=(X=0,Y=0,Z=0,W=1),Translation=(X=0,Y=0,Z=0),Scale3D=(X=0,Y=0,Z=0))
    mirrorTransform.Translation=/Script/RigVM.RigVMFunction_MathVectorScale(Factor=-1, Value=@n1.transform.Translation) {is=struct RigVMUnitNode, template="Scale::Execute(in Value,in Factor,out Result)"}.Result
    mirrorTransform.Rotation=@RigVMFunction_MathQuaternionFromRotatorV2.Result
    mirrorTransform.Scale3D=@n2.Result.Scale
   )
  ]}
 New Function_ContainedGraph {nodes [
   n1 = RigVMFunctionEntryNode()
   > RigVMFunction_Sequence = RigVMAggregateNode()
   Set_Num_2 = RigVMDispatchNode(
    Num=RigVMDispatchNode(Array=@n1.Base Modify Data) {template="DISPATCH_RigVMDispatch_ArrayGetNum(in Array,out Num)"}.Num
    ExecuteContext=@RigVMFunction_Sequence.A
    Array=RigVMVariableNode() {variable=Mirrored}.Value
   ) {template="DISPATCH_RigVMDispatch_ArraySetNum(io Array,in Num)"}
   > For_Each_1 = RigVMDispatchNode(Array=@n1.Base Modify Data) {template="DISPATCH_RigVMDispatch_ArrayIterator(in Array,out Element,out Index,out Count,out Ratio)"}
   > Get Mirror Transform = RigVMFunctionReferenceNode(Transform=@For_Each_1.Element.Transform)
   n2 = RigVMDispatchNode(Index=@For_Each_1.Index, Array=@n1.Mirror Modify Data) {template="DISPATCH_RigVMDispatch_ArrayGetAtIndex(in Array,in Index,out Element)"}
   > Set_At_2 = RigVMDispatchNode(
    Element=(Item=(Type=Bone,Name="None"),Transform=(Rotation=(X=0,Y=0,Z=0,W=1),Translation=(X=0,Y=0,Z=0),Scale3D=(X=1,Y=1,Z=1)))
    Array=RigVMVariableNode() {variable=Mirrored}.Value
    Element.Transform.Rotation=/Script/RigVM.RigVMFunction_MathQuaternionMul(B=@n2.Element.Transform.Rotation, A=@"Get Mirror Transform".mirrorTransform.Rotation) {is=struct RigVMUnitNode, template="Multiply::Execute(in A,in B,out Result)"}.Result
    Element.Item=@n2.Element.Item
    Element.Transform.Scale3D=@"Get Mirror Transform".mirrorTransform.Scale3D
    Index=@For_Each_1.Index
    Element.Transform.Translation=/Script/RigVM.RigVMFunction_MathVectorAdd(B=@n2.Element.Transform.Translation, A=@"Get Mirror Transform".mirrorTransform.Translation) {is=struct RigVMUnitNode, template="Add::Execute(in A,in B,out Result)"}.Result
   ) {template="DISPATCH_RigVMDispatch_ArraySetAtIndex(io Array,in Index,in Element)"}
   &RigUnit_ModifyTransforms(WeightMinimum=0, Weight=@n1.Weight, ExecutePin=@RigVMFunction_Sequence.B, ItemToModify=@n1.Base Modify Data)
   > &RigUnit_ModifyTransforms(WeightMinimum=0, Weight=@n1."Mirror Weight ", ItemToModify=RigVMVariableNode() {variable=Mirrored}.Value)
   RigVMFunctionReturnNode(ExecuteContext=@RigVMFunction_Sequence.C)
  ]}
 New Function_ContainedGraph {nodes [
   n1 = RigVMFunctionEntryNode()
   RigVMFunction_Sequence = &RigVMFunction_Sequence(ExecuteContext=@n1.ExecuteContext)
   RigVMFunction_Sequence_1 = RigVMAggregateNode(ExecuteContext=@RigVMFunction_Sequence.A)
   n2 = &RigVMFunction_Sequence(ExecuteContext=@RigVMFunction_Sequence_1.A)
   For_Each = RigVMDispatchNode(Array=(), ExecuteContext=@n2.B) {template="DISPATCH_RigVMDispatch_ArrayIterator(in Array,out Element,out Index,out Count,out Ratio)"}
   > &RigUnit_SetTransform(
    Weight=1
    bInitial=true
    Item=RigVMDispatchNode(Array=(), Index=@For_Each.Index) {template="DISPATCH_RigVMDispatch_ArrayGetAtIndex(in Array,in Index,out Element)"}.Element
    Value=&RigUnit_GetTransform(bInitial=true, Item=@For_Each.Element).Transform
   )
   For_Each_1 = RigVMDispatchNode(Array=(), ExecuteContext=@For_Each.Completed) {template="DISPATCH_RigVMDispatch_ArrayIterator(in Array,out Element,out Index,out Count,out Ratio)"}
   > &RigUnit_SetTransform(
    Weight=1
    bInitial=true
    Item=RigVMDispatchNode(Array=(), Index=@For_Each_1.Index) {template="DISPATCH_RigVMDispatch_ArrayGetAtIndex(in Array,in Index,out Element)"}.Element
    Value=&RigUnit_GetTransform(bInitial=true, Item=@For_Each_1.Element).Transform
   )
   n3 = &RigUnit_ItemArray(Items=((Type=Bone,Name="dyn_ponytail_base"),(Type=Bone,Name="dyn_ponytail_01"),(Type=Bone,Name="dyn_ponytail_02"),(Type=Bone,Name="dyn_ponytail_03"),(Type=Bone,Name="dyn_ponytail_04"),(Type=Bone,Name="dyn_ponytail_05")))
   RigVMVariableNode(Value=@n3.Items, ExecuteContext=@n2.A) {variable=HairBones}
   n4 = &RigUnit_ItemArray(Items="((Type=Control,Name=\"dyn_ponytail_base_ctrl\"),(Type=Control,Name=\"dyn_ponytail_01_ctrl\"),(Type=Control,Name=\"dyn_ponytail_02_ctrl\"),(Type=Control,Name=\"dyn_ponytail_03_ctrl\"),(Type=Control,Name=\"dyn_ponytail_04_ctrl\"),(Type=Control,Name=\"dyn_ponytail_05_ct ...(truncated)")
   > RigVMVariableNode(Value=@n4.Items) {variable=HairControls}
   &RigUnit_AimConstraintLocalSpaceOffset(
    Child=(Type=Bone,Name="dyn_ponytail_mstr")
    Filter=(bX=True,bY=True,bZ=True)
    Parents=((Item=(Type=Bone,Name="spine_04"),Weight=1))
    Weight=1
    WorldUp=(Target=(X=0,Y=0,Z=1),Kind=Location,Space=(Type=Bone,Name="spine_04"))
    ExecutePin=@RigVMFunction_Sequence_1.B
   )
   > &RigUnit_AimConstraintLocalSpaceOffset(
    Child=(Type=Bone,Name="spline_dyn_ponytail_mstr")
    Filter=(bX=True,bY=True,bZ=True)
    Parents=((Item=(Type=Bone,Name="spine_04"),Weight=1))
    Weight=1
    WorldUp=(Target=(X=0,Y=0,Z=1),Kind=Location,Space=(Type=Bone,Name="spine_04"))
   )
   > &RigUnit_ParentConstraint(
    AdvancedSettings=(RotationOrderForFilter=XZY)
    Child=(Type=Bone,Name="dyn_ponytail_02")
    Parents=((Item=(Type=Bone,Name="head"),Weight=0.7),(Item=(Type=Bone,Name="neck_02"),Weight=0.1),(Item=(Type=Bone,Name="spine_05"),Weight=0.2))
    Weight=1
    bMaintainOffset=true
   )
   > &RigUnit_ParentConstraint(
    AdvancedSettings=(RotationOrderForFilter=XZY)
    Child=(Type=Bone,Name="dyn_ponytail_03")
    Parents=((Item=(Type=Bone,Name="head"),Weight=0.4),(Item=(Type=Bone,Name="neck_02"),Weight=0.1),(Item=(Type=Bone,Name="spine_05"),Weight=0.5))
    Weight=1
    bMaintainOffset=true
   )
   > &RigUnit_ParentConstraint(
    AdvancedSettings=(RotationOrderForFilter=XZY)
    Child=(Type=Bone,Name="dyn_ponytail_04")
    Parents=((Item=(Type=Bone,Name="pelvis"),Weight=0.0593),(Item=(Type=Bone,Name="spine_01"),Weight=0.1075),(Item=(Type=Bone,Name="spine_04"),Weight=0.1383),(Item=(Type=Bone,Name="spine_02"),Weight=0.2488),(Item=(Type=Bone,Name="spine_03"),Weight=0.3369))
    Weight=1
    bMaintainOffset=true
   )
   > &RigUnit_ParentConstraint(
    AdvancedSettings=(RotationOrderForFilter=XZY)
    Child=(Type=Bone,Name="dyn_ponytail_05")
    Parents=((Item=(Type=Bone,Name="pelvis"),Weight=0.2),(Item=(Type=Bone,Name="spine_01"),Weight=0.1),(Item=(Type=Bone,Name="spine_04"),Weight=0.1),(Item=(Type=Bone,Name="spine_02"),Weight=0.4),(Item=(Type=Bone,Name="spine_03"),Weight=0.2))
    Weight=1
    bMaintainOffset=true
   )
   Modify Transforms with Mirror = RigVMFunctionReferenceNode(
    Base Modify Data="((Item=(Type=Bone,Name=\"dyn_ponytail_02\"),Transform=(Rotation=(X=0,Y=0,Z=0,W=-1),Translation=(X=0,Y=1,Z=0),Scale3D=(X=1,Y=1,Z=1))),(Item=(Type=Bone,Name=\"dyn_ponytail_03\"),Transform=(Rotation=(X=0,Y=0,Z=0,W=1),Translation=(X=0,Y=1,Z=0),Scale3D=(X=1,Y=1,Z=1 ...(truncated)"
    Mirror Modify Data="((Item=(Type=Bone,Name=\"spline_ear_r_04\"),Transform=(Rotation=(X=0,Y=0,Z=0,W=1),Translation=(X=0,Y=0,Z=0),Scale3D=(X=1,Y=1,Z=1))),(Item=(Type=Bone,Name=\"spline_ear_r_05\"),Transform=(Rotation=(X=0,Y=0,Z=0,W=1),Translation=(X=0,Y=0,Z=0),Scale3D=(X=1,Y=1,Z=1) ...(truncated)"
    "Mirror Weight "=1
    Weight=1
    ExecuteContext=@RigVMFunction_Sequence_1.D
   )
   > n5 = &RigUnit_SphericalPoseReader(
    ActiveRegionScaleFactors=(PositiveWidth=3,NegativeWidth=1,PositiveHeight=2,NegativeHeight=2)
    ActiveRegionSize=0.1
    Debug=(bDrawDebug=true,bDraw2D=False,bDrawLocalAxes=False,DebugScale=25,DebugSegments=20,DebugThickness=0.25)
    DriverAxis=(X=1,Y=0,Z=0)
    DriverItem=(Type=Bone,Name="head")
    FalloffRegionScaleFactors=(PositiveWidth=1,NegativeWidth=1,PositiveHeight=1,NegativeHeight=1)
    FalloffSize=0.5
    FlipHeightScaling=False
    FlipWidthScaling=False
    OptionalParentItem=(Type=Bone,Name="spine_05")
    RotationOffset=(X=70,Y=0,Z=180)
   )
   > &RigUnit_ModifyTransforms(
    ItemToModify="((Item=(Type=Bone,Name=\"dyn_ponytail_05\"),Transform=(Rotation=(X=0,Y=0,Z=-0.2588,W=0.9659),Translation=(X=1,Y=0,Z=-1),Scale3D=(X=1,Y=1,Z=1))),(Item=(Type=Bone,Name=\"dyn_ponytail_01\"),Transform=(Rotation=(X=0,Y=0,Z=0,W=1),Translation=(X=-5,Y=0,Z=0),Scale3D= ...(truncated)"
    WeightMinimum=0
    Weight=@n5.OutputParam
   )
   SplineIK_1 = RigVMFunctionReferenceNode(
    Primary Axis=(X=0,Y=0,Z=0)
    Samples Per Segment=4
    Secondary Spline Direction=(X=0,Y=0,Z=0)
    Spline Mode=BSpline
    Stretch=true
    Up Axis=(X=0,Y=0,Z=0)
    debug=true
    Bones=RigVMVariableNode() {variable=HairBones}.Value
    Controls=RigVMVariableNode() {variable=HairControls}.Value
    ExecuteContext=@RigVMFunction_Sequence_1.C
   )
   RigVMFunctionReturnNode(ExecuteContext=@RigVMFunction_Sequence.B)
   &RigUnit_ParentConstraint(
    AdvancedSettings=(RotationOrderForFilter=XZY)
    Child=(Type=Bone,Name="dyn_ponytail_01")
    ExecutePin=()
    Parents=((Item=(Type=Bone,Name="head"),Weight=1),(Item=(Type=Bone,Name="neck_02"),Weight=0),(Item=(Type=Bone,Name="spine_05"),Weight=0))
    Weight=1
    bMaintainOffset=true
   )
  ]}
 New Function_ContainedGraph {nodes [
   n1 = RigVMFunctionEntryNode()
   RigVMFunction_Sequence = &RigVMFunction_Sequence(ExecuteContext=@n1.ExecuteContext)
   RigVMFunction_ControlFlowBranch = &RigVMFunction_ControlFlowBranch(ExecuteContext=@RigVMFunction_Sequence.A, Condition=RigVMVariableNode() {variable=RbanOff}.Value)
   n2 = &RigUnit_SphericalPoseReader(
    ActiveRegionScaleFactors=(PositiveWidth=1.5,NegativeWidth=1.5,PositiveHeight=0.75,NegativeHeight=2)
    ActiveRegionSize=0.1
    Debug=(bDrawDebug=true,bDraw2D=False,bDrawLocalAxes=False,DebugScale=25,DebugSegments=20,DebugThickness=0.25)
    DriverAxis=(X=1,Y=0,Z=0)
    DriverItem=(Type=Bone,Name="head")
    FalloffRegionScaleFactors=(PositiveWidth=0.5,NegativeWidth=0.5,PositiveHeight=0.5,NegativeHeight=0.5)
    FalloffSize=0.5
    FlipHeightScaling=False
    FlipWidthScaling=False
    OptionalParentItem=(Type=Bone,Name="spine_05")
    RotationOffset=(X=50,Y=-110,Z=0)
    ExecutePin=@RigVMFunction_ControlFlowBranch.True
   )
   > &RigUnit_ModifyTransforms(
    ItemToModify="((Item=(Type=Bone,Name=\"dyn_ponytail_base\"),Transform=(Rotation=(X=0,Y=0,Z=0.0436,W=0.999),Translation=(X=-1,Y=0,Z=0),Scale3D=(X=1,Y=1,Z=1))),(Item=(Type=Bone,Name=\"dyn_ponytail_01\"),Transform=(Rotation=(X=0,Y=0,Z=0.0872,W=0.9962),Translation=(X=0,Y=0,Z=0) ...(truncated)"
    WeightMinimum=0
    Weight=@n2.OutputParam
   )
   RigVMFunctionReturnNode(ExecuteContext=@RigVMFunction_Sequence.B)
  ]}
 RigVMFunctionLibrary {nodes [
   Get Mirror Transform = RigVMCollapseNode(ExecuteContext=(), transform=(Rotation=(X=0,Y=0,Z=0,W=1),Translation=(X=0,Y=0,Z=0),Scale3D=(X=1,Y=1,Z=1)))
   Hair Normal = RigVMCollapseNode(ExecuteContext=())
   HairSpline = RigVMCollapseNode(ExecuteContext=())
   Modify Transforms with Mirror = RigVMCollapseNode(Base Modify Data=(), ExecuteContext=(), Mirror Modify Data=())
   SplineIK = RigVMCollapseNode(
    Bones=()
    Controls=()
    ExecuteContext=()
    Primary Axis=(X=1,Y=0,Z=0)
    Samples Per Segment=16
    Secondary Spline Direction=(X=0,Y=5,Z=0)
    Spline Mode=BSpline
    Up Axis=(X=0,Y=1,Z=0)
   )
  ]}
 RigVMFunction_Sequence_1_ContainedGraph {nodes [
   n1 = RigVMFunctionEntryNode()
   RigVMFunction_Sequence = &RigVMFunction_Sequence(ExecuteContext=@n1.ExecuteContext)
   n2 = &RigVMFunction_Sequence(ExecuteContext=@RigVMFunction_Sequence.B)
   RigVMFunctionReturnNode(A=@RigVMFunction_Sequence.A, B=@n2.A, C=@n2.B)
  ]}
 RigVMFunction_Sequence_2_ContainedGraph {nodes [
   n1 = RigVMFunctionEntryNode()
   n2 = &RigVMFunction_Sequence(ExecuteContext=@n1.ExecuteContext)
   n3 = &RigVMFunction_Sequence(ExecuteContext=@n2.B)
   n4 = &RigVMFunction_Sequence(ExecuteContext=@n3.B)
   n5 = &RigVMFunction_Sequence(ExecuteContext=@n4.B)
   RigVMFunctionReturnNode(A=@n2.A, B=@n3.A, C=@n4.A, D=@n5.A, E=@n5.B)
  ]}
 RigVMModel {nodes [
   RigUnit_BeginExecution = &RigUnit_BeginExecution()
   RigVMFunction_ControlFlowBranch = &RigVMFunction_ControlFlowBranch(ExecuteContext=@RigUnit_BeginExecution.ExecutePin, Condition=RigVMVariableNode() {variable=UseSplineHair}.Value)
   Hair Normal = RigVMFunctionReferenceNode(ExecuteContext=@RigVMFunction_ControlFlowBranch.False)
   HairSpline = RigVMFunctionReferenceNode(ExecuteContext=@RigVMFunction_ControlFlowBranch.True)
  ]}
]
hierarchy [
 root {loc=<0 0 0>, rot=<0 0 0 1>}
 pelvis {parent=root, loc=<0 -0.5301 98.0571>, rot=<0.0127 0.707 -0.0127 -0.707>}
 spine_01 {parent=pelvis, loc=<6.1208 -0.9328 0>, rot=<0 0 0.034 -0.9994>}
 spine_02 {parent=spine_01, loc=<6.5997 -0.2095 0>, rot=<0 0 0.0238 -0.9997>}
 spine_03 {parent=spine_02, loc=<6.4914 -0.1237 0>, rot=<0 0 0.0149 -0.9999>}
 spine_04 {parent=spine_03, loc=<6.9598 0.0794 0>, rot=<0 0 -0.0132 -0.9999>}
 spine_05 {parent=spine_04, loc=<6.9987 0 0>, rot=<0 0 -0.0015 -1>}
 neck_01 {parent=spine_05, loc=<11.3138 0.1529 0>, rot=<0 0 -0.2003 -0.9797>}
 neck_02 {parent=neck_01, loc=<4.5627 0 0>, rot=<0 0 0.0504 -0.9987>}
 head {parent=neck_02, loc=<5.3914 0 0>, rot=<0 0 0.115 -0.9934>}
 dyn_ponytail_mstr {parent=head, loc=<11.4752 -7.3944 0>, rot=<0.8174 -0.5761 0 0>}
 dyn_ponytail_base {parent=dyn_ponytail_mstr, loc=<0 0 0>, rot=<0 0 0 1>}
 dyn_ponytail_flyaway_top {parent=dyn_ponytail_base, loc=<0 0 0>, rot=<0 0 0 1>}
 dyn_ponytail_01 {parent=dyn_ponytail_base, loc=<4.1715 4.03 0>, rot=<0 0 -0.815 -0.5794>}
 dyn_ponytail_02 {parent=dyn_ponytail_01, loc=<8.2683 -2.3953 0>, rot=<0 0 0 1>}
 dyn_ponytail_03 {parent=dyn_ponytail_02, loc=<8.7875 0 0>, rot=<0 0 0 1>}
 dyn_ponytail_04 {parent=dyn_ponytail_03, loc=<12.541 0 0>, rot=<0 0 0 1>}
 dyn_ponytail_05 {parent=dyn_ponytail_04, loc=<11.4337 0 0>, rot=<0 0 0 1>}
 dyn_ponytail_flyaway_mid_r {parent=dyn_ponytail_02, loc=<6.5786 -3.5987 -2.846>, rot=<0 0 0 1>}
 dyn_ponytail_flyaway_mid_l {parent=dyn_ponytail_02, loc=<6.8518 -2.1527 3.6437>, rot=<0 0 0 1>}
 dyn_ponytail_flyaway_upperR {parent=dyn_ponytail_01, loc=<3.5354 1.0289 -4.4613>, rot=<0 0 0 1>}
 dyn_ponytail_flyaway_upperL {parent=dyn_ponytail_01, loc=<4.4273 1.6105 6.3248>, rot=<0 0 0 1>}
 dyn_bang_r_mstr {parent=head, loc=<11.2893 5.1318 0.9938>, rot=<0.1502 -0.5603 0.7641 0.2824>}
 dyn_bang_r_01 {parent=dyn_bang_r_mstr, loc=<0 0 0>, rot=<0 0 0 1>}
 dyn_bang_r_02 {parent=dyn_bang_r_01, loc=<10.5023 1.9368 4.6403>, rot=<0.4718 0.3005 -0.2275 -0.7971>}
 dyn_bang_r_03 {parent=dyn_bang_r_02, loc=<5.5938 -0.2069 0.1744>, rot=<-0.1552 -0.1556 -0.1396 -0.9655>}
 dyn_bang_r_04 {parent=dyn_bang_r_03, loc=<5.3814 0.7457 0.5435>, rot=<-0.0152 0.1308 -0.0035 -0.9913>}
 dyn_bang_r_front {parent=dyn_bang_r_01, loc=<6.0355 0 0>, rot=<0.0605 0.0707 -0.0043 -0.9956>}
 dyn_bang_l_mstr {parent=head, loc=<11.2897 5.1318 -0.9938>, rot=<-0.1195 0.6379 0.7006 0.2967>}
 dyn_bang_l_01 {parent=dyn_bang_l_mstr, loc=<0 0 0>, rot=<0 0 0 1>}
 dyn_bang_l_02 {parent=dyn_bang_l_01, loc=<10.8038 0.464 -5.0308>, rot=<-0.1855 -0.2153 -0.2745 -0.9186>}
 dyn_bang_l_03 {parent=dyn_bang_l_02, loc=<5.488 -0.1217 -0.3165>, rot=<0.001 0.1195 -0.0555 -0.9913>}
 dyn_bang_l_04 {parent=dyn_bang_l_03, loc=<5.3207 0.6682 -0.3666>, rot=<0.0658 -0.0315 -0.0573 -0.9957>}
 dyn_bang_l_front {parent=dyn_bang_l_01, loc=<7.3655 -1.065 -0.5892>, rot=<0.1746 -0.1243 -0.2286 -0.9496>}
 dyn_bang_mid_front {parent=dyn_bang_l_01, loc=<1.7139 2.0519 2.0209>, rot=<0.6355 0.1168 -0.3036 -0.7002>}
 spline_dyn_ponytail_mstr {parent=head, loc=<11.4752 -7.3944 0>, rot=<0.8174 -0.5761 0 0>}
 spline_dyn_ponytail_base {parent=spline_dyn_ponytail_mstr, loc=<0 0 0>, rot=<0 0 0 1>}
 spline_dyn_ponytail_01 {parent=spline_dyn_ponytail_base, loc=<4.1715 4.03 0>, rot=<0 0 -0.815 -0.5794>}
 spline_dyn_ponytail_02 {parent=spline_dyn_ponytail_01, loc=<8.2683 -2.3953 0>, rot=<0 0 0 1>}
 spline_dyn_ponytail_03 {parent=spline_dyn_ponytail_02, loc=<8.7875 0 0>, rot=<0 0 0 1>}
 spline_dyn_ponytail_04 {parent=spline_dyn_ponytail_03, loc=<12.541 0 0>, rot=<0 0 0 1>}
 spline_dyn_ponytail_05 {parent=spline_dyn_ponytail_04, loc=<11.4337 0 0>, rot=<0 0 0 1>}
 dyn_ponytail_05_ctrl {controlType=EULER_TRANSFORM, parent=spline_dyn_ponytail_05, type=Control, loc=<0 0 0>, rot=<0 0 0 1>}
 dyn_ponytail_04_ctrl {controlType=EULER_TRANSFORM, parent=spline_dyn_ponytail_04, type=Control, loc=<0 0 0>, rot=<0 0 0 1>}
 dyn_ponytail_03_ctrl {controlType=EULER_TRANSFORM, parent=spline_dyn_ponytail_03, type=Control, loc=<0 0 0>, rot=<0 0 0 1>}
 dyn_ponytail_02_ctrl {controlType=EULER_TRANSFORM, parent=spline_dyn_ponytail_02, type=Control, loc=<0 0 0>, rot=<0 0 0 1>}
 dyn_ponytail_01_ctrl {controlType=EULER_TRANSFORM, parent=spline_dyn_ponytail_01, type=Control, loc=<0 0 0>, rot=<0 0 0 1>}
 dyn_ponytail_base_ctrl {controlType=EULER_TRANSFORM, parent=spline_dyn_ponytail_base, type=Control, loc=<0 0 0>, rot=<0 0 0 1>}
]
variables [
 HairBones {
  type="TArray<FRigElementKey>"
  default=((Type=Bone,Name="dyn_ponytail_mstr"),(Type=Bone,Name="dyn_ponytail_base"),(Type=Bone,Name="dyn_ponytail_01"),(Type=Bone,Name="dyn_ponytail_02"),(Type=Bone,Name="dyn_ponytail_03"),(Type=Bone,Name="dyn_ponytail_04"),(Type=Bone,Name="dyn_ponytail_05"))
 }
 HairControls {
  type="TArray<FRigElementKey>"
  default="((Type=Control,Name=\"dyn_ponytail_mstr_ctrl\"),(Type=Control,Name=\"dyn_ponytail_base_ctrl\"),(Type=Control,Name=\"dyn_ponytail_01_ctrl\"),(Type=Control,Name=\"dyn_ponytail_02_ctrl\"),(Type=Control,Name=\"dyn_ponytail_03_ctrl\"),(Type=Control,Name=\"dyn_ponytail_04_ ...(truncated)"
 }
 RbanOff {type=bool, default=False}
 UseSplineHair {type=bool, default=False}
]
