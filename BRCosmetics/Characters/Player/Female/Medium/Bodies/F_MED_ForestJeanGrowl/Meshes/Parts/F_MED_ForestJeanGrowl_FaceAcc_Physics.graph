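# Physics asset for the face accessory part of the ForestJeanGrowl female medium cosmetic body, covering ponytail and bang physics.
# Two four-bone bang chains hang from the head, and a five-bone ponytail chain hangs from base and master bones, with tight swing and twist limits along the strands.
# Capsules on the head, neck, upper spine, clavicles, upper arms and cape and cloak collider bones form the body volume, joined by angularly free constraints.
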PhysicsAsset /BRCosmetics/Characters/Player/Female/Medium/Bodies/F_MED_ForestJeanGrowl/Meshes/Parts/F_MED_ForestJeanGrowl_FaceAcc_Physics
bodies {
 clavicle_l [{shapeType=Capsule, length=0.505, radius=0.505, shapeName=clavicle_l_capsule, rotation=<rot 0 -90 0>}]
 clavicle_r [{shapeType=Capsule, length=0.505, radius=0.505, shapeName=clavicle_r_capsule, rotation=<rot 0 -90 0>}]
 dyn_bang_l_01 [{
   shapeType=Capsule
   length=7.3795
   radius=3.0559
   shapeName=dyn_bang_l_01_capsule
   center=<6.8594 -0.6892 -1.5717>
   rotation=<rot 60.0001 -0.0001 -0.0001>
  }]
 dyn_bang_l_02 [{shapeType=Capsule, length=2.4774, radius=1.9003, shapeName=dyn_bang_l_02_capsule, center=<2.893 0 0>, rotation=<rot 90 0 0>}]
 dyn_bang_l_03 [{shapeType=Capsule, length=3.9557, radius=1.5907, shapeName=dyn_bang_l_03_capsule, center=<2.7096 0 0>, rotation=<rot 90 0 0>}]
 dyn_bang_l_04 [{shapeType=Capsule, length=3.6835, radius=1.6801, shapeName=dyn_bang_l_04_capsule, center=<3.5288 0 0>, rotation=<rot 90 0 0>}]
 dyn_bang_r_01 [{
   shapeType=Capsule
   length=7.3795
   radius=3.0559
   shapeName=dyn_bang_l_01_capsule
   center=<6.8594 -0.6892 -1.5717>
   rotation=<rot 80 179.9999 179.9999>
  }]
 dyn_bang_r_02 [{shapeType=Capsule, length=2.8558, radius=2.2223, shapeName=dyn_bang_r_02_capsule, center=<1.8218 -0.0153 -0.1259>, rotation=<rot 90 0 0>}]
 dyn_bang_r_03 [{shapeType=Capsule, length=3.3818, radius=2.0868, shapeName=dyn_bang_r_03_capsule, center=<1.5768 0.348 -0.2591>, rotation=<rot 90 0 0>}]
 dyn_bang_r_04 [{shapeType=Capsule, length=3.5, radius=1.761, shapeName=dyn_bang_r_04_capsule, center=<2.5129 -0.3646 -0.9758>, rotation=<rot 90 0 0>}]
 dyn_cape_collider [
  {shapeType=Capsule, length=8, radius=12, shapeName=spine_05_capsule, center=<0 12 0>}
  {shapeType=Capsule, length=7, radius=13, shapeName=spine_05_capsule, center=<-12.9543 7.4073 0>}
  {shapeType=Capsule, length=22, radius=8, shapeName=dyn_cape_collider_capsule, center=<-25.69 0 0>, rotation=<rot 0 -0.0001 0>}
 ]
 dyn_cloak_collider [{shapeType=Capsule, length=17.005, radius=6.005, shapeName=spine_05_capsule, center=<0.7283 0.4475 0>, rotation=<rot 0 -90 0>}]
 dyn_ponytail_01 [{
   shapeType=Capsule
   length=9.278
   radius=6.4559
   shapeName=dyn_ponytail_01_capsule
   center=<7.4035 -1.5 0>
   rotation=<rot -0.2119 -94.9995 -89.2587>
  }]
 dyn_ponytail_02 [{
   shapeType=Capsule
   length=9.278
   radius=6.4559
   shapeName=dyn_ponytail_01_capsule
   center=<7.5303 -0.1216 0.6172>
   rotation=<rot -0.2757 -90 -89.28>
  }]
 dyn_ponytail_03 [{
   shapeType=Capsule
   length=9.278
   radius=6.4559
   shapeName=dyn_ponytail_01_capsule
   center=<7.5303 -0.1216 0.6172>
   rotation=<rot -0.2757 -90 -89.28>
  }]
 dyn_ponytail_04 [{
   shapeType=Capsule
   length=10.278
   radius=4.9559
   shapeName=dyn_ponytail_01_capsule
   center=<7.5303 -0.1216 0.6172>
   rotation=<rot -0.2757 -90 -89.28>
  }]
 dyn_ponytail_05 [{
   shapeType=Capsule
   length=9.778
   radius=3.9559
   shapeName=dyn_ponytail_01_capsule
   center=<7.5303 -0.1216 0.6172>
   rotation=<rot -0.2757 -90 -89.28>
  }]
 dyn_ponytail_base [{shapeType=Capsule, length=4.278, radius=2.9559, shapeName=dyn_ponytail_01_capsule, rotation=<rot 90 0 0>}]
 dyn_ponytail_mstr [{shapeType=Capsule, length=0.505, radius=0.505, shapeName=dyn_ponytail_mstr_capsule, rotation=<rot 0 -90 0>}]
 head [{shapeType=Capsule, length=9.2386, radius=5.6277, center=<3.2489 -0.1429 0>, rotation=<rot 90 0 0>}]
 neck_01 [{shapeType=Capsule, length=4, radius=4.5, shapeName=neck_01_capsule, rotation=<rot 90 -10.0001 0>}]
 neck_02 [{shapeType=Capsule, length=0.505, radius=0.505, shapeName=neck_02_capsule, rotation=<rot 0 -90 0>}]
 spine_03 [{shapeType=Capsule, length=9.005, radius=10.005, shapeName=spine_05_capsule, center=<-5.3157 1.8056 0>, rotation=<rot 0 -90 0>}]
 spine_04 [{shapeType=Capsule, length=12.505, radius=10.005, shapeName=spine_05_capsule, center=<-2.319 3.2469 0>, rotation=<rot 0 -90 0>}]
 spine_05 [{shapeType=Capsule, length=22.505, radius=7.505, shapeName=spine_05_capsule, center=<0.7283 0.4475 0>, rotation=<rot 0 -90 0>}]
 upperarm_l [{shapeType=Capsule, length=22, radius=4.505, shapeName=upperarm_l_capsule, center=<14.0961 0 0>, rotation=<rot -0.0001 -89.9999 90>}]
 upperarm_r [{shapeType=Capsule, length=22, radius=4.505, shapeName=upperarm_l_capsule, center=<-15.3282 0 0>, rotation=<rot -0.0001 -89.9999 90>}]
}
constraints [
 {bone1Name=dyn_ponytail_02, bone2Name=dyn_ponytail_01, swing1LimitDegrees=30, swing2LimitDegrees=30, twistLimitDegrees=5}
 {bone1Name=dyn_ponytail_04, bone2Name=dyn_ponytail_03, swing1LimitDegrees=30, swing2LimitDegrees=30, twistLimitDegrees=5}
 {
  bone1Name=head
  bone2Name=neck_02
  swing1LimitDegrees=45
  swing1Motion=Free
  swing2LimitDegrees=45
  swing2Motion=Free
  twistLimitDegrees=45
  twistMotion=Free
 }
 {
  bone1Name=neck_02
  bone2Name=neck_01
  swing1LimitDegrees=45
  swing1Motion=Free
  swing2LimitDegrees=45
  swing2Motion=Free
  twistLimitDegrees=45
  twistMotion=Free
 }
 {
  bone1Name=dyn_ponytail_mstr
  bone2Name=head
  swing1LimitDegrees=45
  swing1Motion=Free
  swing2LimitDegrees=45
  swing2Motion=Free
  twistLimitDegrees=45
  twistMotion=Free
 }
 {bone1Name=dyn_ponytail_01, bone2Name=dyn_ponytail_base, swing1LimitDegrees=30, swing2LimitDegrees=30, twistLimitDegrees=5}
 {bone1Name=dyn_ponytail_03, bone2Name=dyn_ponytail_02, swing1LimitDegrees=30, swing2LimitDegrees=30, twistLimitDegrees=5}
 {bone1Name=dyn_ponytail_05, bone2Name=dyn_ponytail_04, swing1LimitDegrees=30, swing2LimitDegrees=30, twistLimitDegrees=5}
 {bone1Name=dyn_ponytail_base, bone2Name=dyn_ponytail_mstr, swing1LimitDegrees=20, swing2LimitDegrees=20, twistLimitDegrees=0}
 {
  bone1Name=spine_05
  bone2Name=spine_04
  swing1LimitDegrees=45
  swing1Motion=Free
  swing2LimitDegrees=45
  swing2Motion=Free
  twistLimitDegrees=45
  twistMotion=Free
 }
 {
  bone1Name=spine_04
  bone2Name=spine_03
  swing1LimitDegrees=45
  swing1Motion=Free
  swing2LimitDegrees=45
  swing2Motion=Free
  twistLimitDegrees=45
  twistMotion=Free
 }
 {bone1Name=dyn_bang_r_04, bone2Name=dyn_bang_r_03, swing1LimitDegrees=25, swing2LimitDegrees=25, twistLimitDegrees=5}
 {bone1Name=dyn_bang_r_01, bone2Name=head, swing1LimitDegrees=10, swing2LimitDegrees=10, twistLimitDegrees=5}
 {bone1Name=dyn_bang_r_02, bone2Name=dyn_bang_r_01, swing1LimitDegrees=25, swing2LimitDegrees=25, twistLimitDegrees=5}
 {bone1Name=dyn_bang_r_03, bone2Name=dyn_bang_r_02, swing1LimitDegrees=25, swing2LimitDegrees=25, twistLimitDegrees=5}
 {bone1Name=dyn_bang_l_04, bone2Name=dyn_bang_l_03, swing1LimitDegrees=25, swing2LimitDegrees=25, twistLimitDegrees=5}
 {bone1Name=dyn_bang_l_01, bone2Name=head, swing1LimitDegrees=10, swing2LimitDegrees=10, twistLimitDegrees=5}
 {bone1Name=dyn_bang_l_02, bone2Name=dyn_bang_l_01, swing1LimitDegrees=25, swing2LimitDegrees=25, twistLimitDegrees=5}
 {bone1Name=dyn_bang_l_03, bone2Name=dyn_bang_l_02, swing1LimitDegrees=25, swing2LimitDegrees=25, twistLimitDegrees=5}
 {
  bone1Name=neck_01
  bone2Name=spine_05
  swing1LimitDegrees=45
  swing1Motion=Free
  swing2LimitDegrees=45
  swing2Motion=Free
  twistLimitDegrees=45
  twistMotion=Free
 }
 {
  bone1Name=dyn_cape_collider
  bone2Name=spine_05
  swing1LimitDegrees=45
  swing1Motion=Free
  swing2LimitDegrees=45
  swing2Motion=Free
  twistLimitDegrees=45
  twistMotion=Free
 }
 {
  bone1Name=dyn_cloak_collider
  bone2Name=spine_05
  swing1LimitDegrees=45
  swing1Motion=Free
  swing2LimitDegrees=45
  swing2Motion=Free
  twistLimitDegrees=45
  twistMotion=Free
 }
 {
  bone1Name=clavicle_l
  bone2Name=spine_05
  swing1LimitDegrees=45
  swing1Motion=Free
  swing2LimitDegrees=45
  swing2Motion=Free
  twistLimitDegrees=45
  twistMotion=Free
 }
 {
  bone1Name=clavicle_r
  bone2Name=spine_05
  swing1LimitDegrees=45
  swing1Motion=Free
  swing2LimitDegrees=45
  swing2Motion=Free
  twistLimitDegrees=45
  twistMotion=Free
 }
 {
  bone1Name=upperarm_l
  bone2Name=clavicle_l
  swing1LimitDegrees=45
  swing1Motion=Free
  swing2LimitDegrees=45
  swing2Motion=Free
  twistLimitDegrees=45
  twistMotion=Free
 }
 {
  bone1Name=upperarm_r
  bone2Name=clavicle_r
  swing1LimitDegrees=45
  swing1Motion=Free
  swing2LimitDegrees=45
  swing2Motion=Free
  twistLimitDegrees=45
  twistMotion=Free
 }
]
