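# Animation Blueprint for the ForestJeanGrowl female character's face accessory, driving secondary motion on ponytail, flyaway and bang bones.
# A Control Rig over the pose copied from the parent mesh feeds cached branches, selected by backbling and rigid-body flags.
# One pairs RigidBody with Anim Dynamics; its fallback uses Anim Dynamics chains with crouch-driven constraints.
# The backbling branch adds pose adjustments, while cape, movement and front-end states toggle corrective offsets and constraints.
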AnimBlueprint /BRCosmetics/Characters/Player/Female/Medium/Bodies/F_MED_ForestJeanGrowl/Meshes/Parts/F_MED_ForestJeanGrowl_FaceAcc_AnimBP
target_skeleton=&/F_MED_ForestJeanGrowl_FaceAcc_Skeleton
graphs [
 AnimGraph {nodes [
   n1 = Use cached pose 'RBanPose'()
   n2 = Use cached pose 'NoRBAN'()
   n3 = Get bRigidBodyEnabled()
   Output Pose(Result=Blend Poses by bool(
     BlendPose_0=Use cached pose 'BackpackPose'()
     BlendPose_1=Blend Poses by bool(BlendPose_0=@n1, BlendPose_1=@n2, BlendTime_0=0, BlendTime_1=0, bActiveValue=@n3) {blendType=Linear}
     BlendTime_0=0
     BlendTime_1=0
     bActiveValue=Get Is Wearing Backbling()
    ) {blendType=Linear}) {layerGroup=None}
   n4 = Anim Dynamics - Bone: dyn_ponytail_flyaway_top!(SimulationSpace:AnimPhysSimSpaceType Enum, ComponentPose=Local To Component(LocalPose=Use cached pose 'InitialPose'()), Alpha=1) -> (Pose) {
    angularDampingOverride=1
    angularSpringConstant=23000
    bAngularSpring=true
    bOverrideAngularDamping=true
    bUseGravityOverride=true
    gravityScale=0.35
    simulationSpace=RootRelative
    alphaBoolBlend {blendOption=Linear}
    boundBone {boneName=dyn_ponytail_flyaway_top}
    componentAppliedLinearAccClamp=<2000 2000 2000>
    componentLinearAccScale=<1 1 1>
    componentLinearVelScale=<1 1 1>
    gravityOverride {z=-980}
    physicsBodyDefinitions [{
      boundBone {boneName=dyn_ponytail_flyaway_top}
      boxExtents=<12 12 12>
      constraintSetup {angularLimitsMax=<15 15 15>, angularLimitsMin=<-15 -15 -25>, angularTarget=<1 0 0>}
      localJointOffset=<1 0 0>
     }]
   }
   n5 = Anim Dynamics - Bone: dyn_ponytail_01!(SimulationSpace:AnimPhysSimSpaceType Enum, ComponentPose=@n4, Alpha=1, ExternalForce="0,0,0") -> (Pose) {
    angularDampingOverride=1
    angularSpringConstant=600
    bAngularSpring=true
    bOverrideAngularDamping=true
    bUseGravityOverride=true
    gravityScale=0.35
    simulationSpace=RootRelative
    alphaBoolBlend {blendOption=Linear}
    boundBone {boneName=dyn_ponytail_01}
    componentAppliedLinearAccClamp=<1300 1300 1300>
    componentLinearAccScale=<1 1 1>
    componentLinearVelScale=<1 1 1>
    gravityOverride {z=-980}
    physicsBodyDefinitions [{
      boundBone {boneName=dyn_ponytail_01}
      boxExtents=<12 12 12>
      constraintSetup {angularLimitsMax=<5 25 0>, angularLimitsMin=<-5 -25 -40>, angularTarget=<1 0 0>}
      localJointOffset=<1 0 0>
     }]
   }
   n6 = Anim Dynamics - Bone: dyn_ponytail_02!(SimulationSpace:AnimPhysSimSpaceType Enum, ComponentPose=@n5, Alpha=1, ExternalForce="0,0,0") -> (Pose) {
    angularDampingOverride=1
    angularSpringConstant=600
    bAngularSpring=true
    bOverrideAngularDamping=true
    bUseGravityOverride=true
    gravityScale=0.35
    simulationSpace=RootRelative
    alphaBoolBlend {blendOption=Linear}
    boundBone {boneName=dyn_ponytail_02}
    componentAppliedLinearAccClamp=<1300 1300 1300>
    componentLinearAccScale=<1 1 1>
    componentLinearVelScale=<1 1 1>
    gravityOverride {z=-980}
    physicsBodyDefinitions [{
      boundBone {boneName=dyn_ponytail_02}
      boxExtents=<12 12 12>
      constraintSetup {angularLimitsMax=<5 25 0>, angularLimitsMin=<-5 -25 -40>, angularTarget=<1 0 0>}
      localJointOffset=<1 0 0>
     }]
   }
   n7 = Anim Dynamics - Bone: dyn_ponytail_03!(SimulationSpace:AnimPhysSimSpaceType Enum, ComponentPose=@n6, Alpha=1, ExternalForce="0,0,0") -> (Pose) {
    angularDampingOverride=1
    angularSpringConstant=600
    bAngularSpring=true
    bOverrideAngularDamping=true
    bUseGravityOverride=true
    gravityScale=0.35
    simulationSpace=RootRelative
    alphaBoolBlend {blendOption=Linear}
    boundBone {boneName=dyn_ponytail_03}
    componentAppliedLinearAccClamp=<1300 1300 1600>
    componentLinearAccScale=<1 1 1>
    componentLinearVelScale=<1 1 1>
    gravityOverride {z=-980}
    physicsBodyDefinitions [{
      boundBone {boneName=dyn_ponytail_03}
      boxExtents=<12 12 12>
      constraintSetup {angularLimitsMax=<5 25 0>, angularLimitsMin=<-5 -25 -40>, angularTarget=<1 0 0>}
      localJointOffset=<1 0 0>
     }]
   }
   n8 = Anim Dynamics - Bone: dyn_ponytail_04!(SimulationSpace:AnimPhysSimSpaceType Enum, ComponentPose=@n7, Alpha=1, ExternalForce="0,0,0") -> (Pose) {
    angularDampingOverride=1
    angularSpringConstant=600
    bAngularSpring=true
    bOverrideAngularDamping=true
    bUseGravityOverride=true
    gravityScale=0.35
    simulationSpace=RootRelative
    alphaBoolBlend {blendOption=Linear}
    boundBone {boneName=dyn_ponytail_04}
    componentAppliedLinearAccClamp=<1300 1300 2400>
    componentLinearAccScale=<1 1 1>
    componentLinearVelScale=<1 1 1>
    gravityOverride {z=-980}
    physicsBodyDefinitions [{
      boundBone {boneName=dyn_ponytail_04}
      boxExtents=<12 12 12>
      constraintSetup {angularLimitsMax=<5 25 0>, angularLimitsMin=<-5 -25 -40>, angularTarget=<1 0 0>}
      localJointOffset=<1 0 0>
     }]
   }
   n9 = Anim Dynamics - Bone: dyn_ponytail_05!(SimulationSpace:AnimPhysSimSpaceType Enum, ComponentPose=@n8, Alpha=1, ExternalForce="0,0,0") -> (Pose) {
    angularDampingOverride=1
    angularSpringConstant=600
    bAngularSpring=true
    bOverrideAngularDamping=true
    bUseGravityOverride=true
    gravityScale=0.35
    simulationSpace=RootRelative
    alphaBoolBlend {blendOption=Linear}
    boundBone {boneName=dyn_ponytail_05}
    componentAppliedLinearAccClamp=<1300 1300 2400>
    componentLinearAccScale=<1 1 1>
    componentLinearVelScale=<1 1 1>
    gravityOverride {z=-980}
    physicsBodyDefinitions [{
      boundBone {boneName=dyn_ponytail_05}
      boxExtents=<12 12 12>
      constraintSetup {angularLimitsMax=<5 25 0>, angularLimitsMin=<-5 -25 -40>, angularTarget=<1 0 0>}
      localJointOffset=<1 0 0>
     }]
   }
   n10 = Anim Dynamics - Bone: dyn_ponytail_flyaway_upperL!(SimulationSpace:AnimPhysSimSpaceType Enum, ComponentPose=@n9, Alpha=1) -> (Pose) {
    angularDampingOverride=1
    angularSpringConstant=2000
    bAngularSpring=true
    bOverrideAngularDamping=true
    bUseGravityOverride=true
    gravityScale=0.35
    simulationSpace=RootRelative
    alphaBoolBlend {blendOption=Linear}
    boundBone {boneName=dyn_ponytail_flyaway_upperL}
    componentAppliedLinearAccClamp=<2000 2000 2000>
    componentLinearAccScale=<1 1 1>
    componentLinearVelScale=<1 1 1>
    gravityOverride {z=-980}
    physicsBodyDefinitions [{
      boundBone {boneName=dyn_ponytail_flyaway_upperL}
      boxExtents=<30 30 30>
      constraintSetup {angularLimitsMax=<15 15 15>, angularLimitsMin=<-15 -15 -15>, angularTarget=<1 0 0>}
      localJointOffset=<3 0 0>
     }]
   }
   n11 = Anim Dynamics - Bone: dyn_ponytail_flyaway_mid_l!(SimulationSpace:AnimPhysSimSpaceType Enum, ComponentPose=@n10, Alpha=1) -> (Pose) {
    angularDampingOverride=1
    angularSpringConstant=2000
    bAngularSpring=true
    bOverrideAngularDamping=true
    bUseGravityOverride=true
    gravityScale=0.35
    simulationSpace=RootRelative
    alphaBoolBlend {blendOption=Linear}
    boundBone {boneName=dyn_ponytail_flyaway_mid_l}
    componentAppliedLinearAccClamp=<2000 2000 2000>
    componentLinearAccScale=<1 1 1>
    componentLinearVelScale=<1 1 1>
    gravityOverride {z=-980}
    physicsBodyDefinitions [{
      boundBone {boneName=dyn_ponytail_flyaway_mid_l}
      boxExtents=<30 30 30>
      constraintSetup {angularLimitsMax=<15 15 15>, angularLimitsMin=<-15 -15 -15>, angularTarget=<1 0 0>}
      localJointOffset=<3 0 0>
     }]
   }
   n12 = Anim Dynamics - Bone: dyn_ponytail_flyaway_upperR!(SimulationSpace:AnimPhysSimSpaceType Enum, ComponentPose=@n11, Alpha=1) -> (Pose) {
    angularDampingOverride=1
    angularSpringConstant=2000
    bAngularSpring=true
    bOverrideAngularDamping=true
    bUseGravityOverride=true
    gravityScale=0.35
    simulationSpace=RootRelative
    alphaBoolBlend {blendOption=Linear}
    boundBone {boneName=dyn_ponytail_flyaway_upperR}
    componentAppliedLinearAccClamp=<2000 2000 2000>
    componentLinearAccScale=<1 1 1>
    componentLinearVelScale=<1 1 1>
    gravityOverride {z=-980}
    physicsBodyDefinitions [{
      boundBone {boneName=dyn_ponytail_flyaway_upperR}
      boxExtents=<30 30 30>
      constraintSetup {angularLimitsMax=<15 15 15>, angularLimitsMin=<-15 -15 -15>, angularTarget=<1 0 0>}
      localJointOffset=<3 0 0>
     }]
   }
   n13 = Anim Dynamics - Bone: dyn_ponytail_flyaway_mid_r!(SimulationSpace:AnimPhysSimSpaceType Enum, ComponentPose=@n12, Alpha=1) -> (Pose) {
    angularDampingOverride=1
    angularSpringConstant=2000
    bAngularSpring=true
    bOverrideAngularDamping=true
    bUseGravityOverride=true
    gravityScale=0.35
    simulationSpace=RootRelative
    alphaBoolBlend {blendOption=Linear}
    boundBone {boneName=dyn_ponytail_flyaway_mid_r}
    componentAppliedLinearAccClamp=<2000 2000 2000>
    componentLinearAccScale=<1 1 1>
    componentLinearVelScale=<1 1 1>
    gravityOverride {z=-980}
    physicsBodyDefinitions [{
      boundBone {boneName=dyn_ponytail_flyaway_mid_r}
      boxExtents=<30 30 30>
      constraintSetup {angularLimitsMax=<15 15 15>, angularLimitsMin=<-15 -15 -15>, angularTarget=<1 0 0>}
      localJointOffset=<3 0 0>
     }]
   }
   n14 = Anim Dynamics - Bone: dyn_bang_mid_front!(SimulationSpace:AnimPhysSimSpaceType Enum, ComponentPose=@n13, Alpha=1, ExternalForce="0,0,0") -> (Pose) {
    angularDampingOverride=1
    bOverrideAngularDamping=true
    bUseGravityOverride=true
    simulationSpace=RootRelative
    alphaBoolBlend {blendOption=Linear}
    boundBone {boneName=dyn_bang_mid_front}
    componentAppliedLinearAccClamp=<2000 2000 2000>
    componentLinearAccScale=<1 1 1>
    componentLinearVelScale=<1 1 1>
    gravityOverride {z=-980}
    physicsBodyDefinitions [{
      boundBone {boneName=dyn_bang_mid_front}
      boxExtents=<12 12 12>
      constraintSetup {angularLimitsMax=<1 20 2>, angularLimitsMin=<-1 -20 -25>, angularTarget=<0 0 0>}
      localJointOffset=<1 0 0>
     }]
    simSpaceSettings {maxAngularAcceleration=35, maxAngularVelocity=7, simSpaceAngularAlpha=1}
   }
   n15 = Anim Dynamics - Bone: dyn_bang_l_front!(SimulationSpace:AnimPhysSimSpaceType Enum, ComponentPose=@n14, Alpha=1) -> (Pose) {
    angularDampingOverride=1
    angularSpringConstant=2000
    bAngularSpring=true
    bOverrideAngularDamping=true
    bUseGravityOverride=true
    gravityScale=0.35
    simulationSpace=RootRelative
    alphaBoolBlend {blendOption=Linear}
    boundBone {boneName=dyn_bang_l_front}
    componentAppliedLinearAccClamp=<2000 2000 2000>
    componentLinearAccScale=<1 1 1>
    componentLinearVelScale=<1 1 1>
    gravityOverride {z=-980}
    physicsBodyDefinitions [{
      boundBone {boneName=dyn_bang_l_front}
      boxExtents=<30 30 30>
      constraintSetup {angularLimitsMax=<15 5 5>, angularLimitsMin=<-15 -45 -45>, angularTarget=<1 0 0>}
      localJointOffset=<3 0 0>
     }]
   }
   n16 = Anim Dynamics - Bone: dyn_bang_r_front!(SimulationSpace:AnimPhysSimSpaceType Enum, ComponentPose=@n15, Alpha=1) -> (Pose) {
    angularDampingOverride=1
    angularSpringConstant=2000
    bAngularSpring=true
    bOverrideAngularDamping=true
    bUseGravityOverride=true
    gravityScale=0.35
    simulationSpace=RootRelative
    alphaBoolBlend {blendOption=Linear}
    boundBone {boneName=dyn_bang_r_front}
    componentAppliedLinearAccClamp=<2000 2000 2000>
    componentLinearAccScale=<1 1 1>
    componentLinearVelScale=<1 1 1>
    gravityOverride {z=-980}
    physicsBodyDefinitions [{
      boundBone {boneName=dyn_bang_r_front}
      boxExtents=<30 30 30>
      constraintSetup {angularLimitsMax=<15 30 0>, angularLimitsMin=<-5 -5 -15>, angularTarget=<1 0 0>}
      localJointOffset=<3 0 0>
     }]
   }
   n17 = Anim Dynamics - Bone: dyn_bang_l_02!(SimulationSpace:AnimPhysSimSpaceType Enum, ComponentPose=@n16, Alpha=1) -> (Pose) {
    angularDampingOverride=1
    angularSpringConstant=2000
    bAngularSpring=true
    bOverrideAngularDamping=true
    bUseGravityOverride=true
    gravityScale=0.35
    simulationSpace=RootRelative
    alphaBoolBlend {blendOption=Linear}
    boundBone {boneName=dyn_bang_l_02}
    componentAppliedLinearAccClamp=<2000 2000 2000>
    componentLinearAccScale=<1 1 1>
    componentLinearVelScale=<1 1 1>
    gravityOverride {z=-980}
    physicsBodyDefinitions [{
      boundBone {boneName=dyn_bang_l_02}
      boxExtents=<30 30 30>
      constraintSetup {angularLimitsMax=<15 15 5>, angularLimitsMin=<-15 -45 -45>, angularTarget=<1 0 0>}
      localJointOffset=<3 0 0>
     }]
   }
   n18 = Anim Dynamics - Bone: dyn_bang_r_02!(SimulationSpace:AnimPhysSimSpaceType Enum, ComponentPose=@n17, Alpha=1) -> (Pose) {
    angularDampingOverride=1
    angularSpringConstant=2000
    bAngularSpring=true
    bOverrideAngularDamping=true
    bUseGravityOverride=true
    gravityScale=0.35
    simulationSpace=RootRelative
    alphaBoolBlend {blendOption=Linear}
    boundBone {boneName=dyn_bang_r_02}
    componentAppliedLinearAccClamp=<2000 2000 2000>
    componentLinearAccScale=<1 1 1>
    componentLinearVelScale=<1 1 1>
    gravityOverride {z=-980}
    physicsBodyDefinitions [{
      boundBone {boneName=dyn_bang_r_02}
      boxExtents=<30 30 30>
      constraintSetup {angularLimitsMax=<15 30 0>, angularLimitsMin=<-5 -15 -15>, angularTarget=<1 0 0>}
      localJointOffset=<3 0 0>
     }]
   }
   n19 = Anim Dynamics - Bone: dyn_bang_l_03!(SimulationSpace:AnimPhysSimSpaceType Enum, ComponentPose=@n18, Alpha=1) -> (Pose) {
    angularDampingOverride=1
    angularSpringConstant=2000
    bAngularSpring=true
    bOverrideAngularDamping=true
    bUseGravityOverride=true
    gravityScale=0.35
    simulationSpace=RootRelative
    alphaBoolBlend {blendOption=Linear}
    boundBone {boneName=dyn_bang_l_03}
    componentAppliedLinearAccClamp=<2000 2000 2000>
    componentLinearAccScale=<1 1 1>
    componentLinearVelScale=<1 1 1>
    gravityOverride {z=-980}
    physicsBodyDefinitions [{
      boundBone {boneName=dyn_bang_l_03}
      boxExtents=<30 30 30>
      constraintSetup {angularLimitsMax=<15 15 5>, angularLimitsMin=<-15 -45 -45>, angularTarget=<1 0 0>}
      localJointOffset=<3 0 0>
     }]
   }
   n20 = Anim Dynamics - Bone: dyn_bang_r_03!(SimulationSpace:AnimPhysSimSpaceType Enum, ComponentPose=@n19, Alpha=1) -> (Pose) {
    angularDampingOverride=1
    angularSpringConstant=2000
    bAngularSpring=true
    bOverrideAngularDamping=true
    bUseGravityOverride=true
    gravityScale=0.35
    simulationSpace=RootRelative
    alphaBoolBlend {blendOption=Linear}
    boundBone {boneName=dyn_bang_r_03}
    componentAppliedLinearAccClamp=<2000 2000 2000>
    componentLinearAccScale=<1 1 1>
    componentLinearVelScale=<1 1 1>
    gravityOverride {z=-980}
    physicsBodyDefinitions [{
      boundBone {boneName=dyn_bang_r_03}
      boxExtents=<30 30 30>
      constraintSetup {angularLimitsMax=<15 30 0>, angularLimitsMin=<-5 -15 -15>, angularTarget=<1 0 0>}
      localJointOffset=<3 0 0>
     }]
   }
   n21 = Get Crouch() -> (Crouch:double)
   n22 = "Constraint - dyn_ponytail_03 (PPP)"!(ComponentPose=@n20, Alpha=@n21, ConstraintWeights_0=0.2, ConstraintWeights_1:float=0.8, ConstraintWeights_2:float=0) -> (Pose) {
    alphaScaleBiasClamp {bInterpResult=true, interpSpeedDecreasing=7, interpSpeedIncreasing=5}
    boneToModify {boneName=dyn_ponytail_03}
    constraintSetup [{targetBone {boneName=neck_01}}, {targetBone {boneName=spine_05}}, {targetBone {boneName=head}}]
    constraintWeights [1,1,1]
   }
   n23 = "Constraint - dyn_ponytail_04 (PP)"!(ComponentPose=@n22, Alpha=@n21, ConstraintWeights_0=1, ConstraintWeights_1:float=0) -> (Pose) {
    alphaScaleBiasClamp {bInterpResult=true, interpSpeedDecreasing=7, interpSpeedIncreasing=5}
    boneToModify {boneName=dyn_ponytail_04}
    constraintSetup [{targetBone {boneName=spine_04}}, {targetBone {boneName=neck_01}}]
    constraintWeights [1,1]
   }
   n24 = "Constraint - dyn_ponytail_05 (PP)"!(ComponentPose=@n23, Alpha=@n21, ConstraintWeights_0=1, ConstraintWeights_1:float=0) -> (Pose) {
    alphaScaleBiasClamp {bInterpResult=true, interpSpeedDecreasing=7, interpSpeedIncreasing=5}
    boneToModify {boneName=dyn_ponytail_05}
    constraintSetup [{targetBone {boneName=spine_03}}, {targetBone {boneName=neck_01}}]
    constraintWeights [1,1]
   }
   n25 = Get Is Backbling A Cape() -> (Is Backbling A Cape:bool)
   n26 = "Transform (Modify) Bone - Bone: dyn_ponytail_base"!(ComponentPose=@n24, bAlphaBoolEnabled:bool=@n25, Translation="0,-1,0", Rotation="0,0,10", Scale="1,1,1") -> (Pose) {
    alphaInputType=Bool
    rotationMode=BMM_Additive
    scaleMode=BMM_Additive
    translationMode=BMM_Additive
    boneToModify {boneName=dyn_ponytail_base}
    rotation {roll=10}
    translation {y=-1}
   }
   n27 = "Transform (Modify) Bone - Bone: dyn_ponytail_03"!(ComponentPose=@n26, bAlphaBoolEnabled:bool=@n25, Translation="0,-1,0", Rotation="0,0,10", Scale="1,1,1") -> (Pose) {
    alphaInputType=Bool
    rotationMode=BMM_Additive
    scaleMode=BMM_Additive
    translationMode=BMM_Additive
    boneToModify {boneName=dyn_ponytail_03}
    rotation {roll=10}
    translation {y=-1}
   }
   Save cached pose 'NoRBAN'(Pose=Component To Local(ComponentPose=@n27)) {cacheName=NoRBAN}
   n28 = Control Rig!(
    Source=Copy Pose From Mesh() {bUseAttachedParent=true}
    Alpha
    UseSplineHair:bool=Get Is Wearing Backbling()
    RbanOff:bool=NOT Boolean(A=Get bRigidBodyEnabled())
   ) -> (Pose) {controlRigAssetReference {blueprintRigClass=&/F_MED_ForestJeanGrowl_FaceAcc_CtrlRig.F_MED_ForestJeanGrowl_FaceAcc_CtrlRig_C}}
   Save cached pose 'InitialPose'(Pose=@n28) {cacheName=InitialPose}
   n29 = "Transform (Modify) Bone - Bone: dyn_cape_collider"(
    ComponentPose=Local To Component(LocalPose=Use cached pose 'InitialPose'())
    Alpha=Get Moving() -> (Moving:double)
    Translation="-2,-4,1"
    Rotation="0,30,0"
    Scale="1,1,1"
   ) {
    rotationMode=BMM_Additive
    rotationSpace=BCS_BoneSpace
    translationMode=BMM_Additive
    translationSpace=BCS_BoneSpace
    alphaBoolBlend {blendOption=Linear}
    alphaScaleBiasClamp {bInterpResult=true, interpSpeedDecreasing=7, interpSpeedIncreasing=5}
    boneToModify {boneName=dyn_cape_collider}
    rotation {yaw=30}
    translation=<-2 -4 1>
   }
   n30 = "Transform (Modify) Bone - Bone: dyn_cape_collider"!(ComponentPose=@n29, bAlphaBoolEnabled:bool=True, Translation="-6,-14,2", Rotation="-0.004,-20,0", Scale="1,1,1") -> (Pose) {
    alphaInputType=Bool
    rotationMode=BMM_Additive
    rotationSpace=BCS_BoneSpace
    translationMode=BMM_Additive
    translationSpace=BCS_BoneSpace
    alphaBoolBlend {blendOption=Linear}
    boneToModify {boneName=dyn_cape_collider}
    rotation {pitch=-0.004, yaw=-20}
    translation=<-6 -14 2>
   }
   n31 = Component To Local(ComponentPose=@n30)
   n32 = Use cached pose 'InitialPose'()
   n33 = Get Is Backbling A Cape()
   n34 = "Transform (Modify) Bone - Bone: dyn_cloak_collider"!(
    ComponentPose=Local To Component(LocalPose=Blend Poses by bool(BlendPose_0=@n31, BlendPose_1=@n32, BlendTime_0=0, BlendTime_1=0, bActiveValue=@n33) {blendType=Linear})
    bAlphaBoolEnabled:bool=Get ForestJeanGrowl_Joy() -> (ForestJeanGrowl_Joy:bool)
    Translation="5,0,0"
    Rotation="0,0,0"
    Scale="1,1,1"
   ) -> (Pose) {
    alphaInputType=Bool
    translationMode=BMM_Additive
    translationSpace=BCS_BoneSpace
    boneToModify {boneName=dyn_cloak_collider}
    translation {x=5}
   }
   n35 = "Constraint - dyn_bang_l_04 (PPP)"!(
    ComponentPose=@n34
    bAlphaBoolEnabled:bool=Get bIsInFrontEnd()
    ConstraintWeights_0=0.5
    ConstraintWeights_1:float=0.25
    ConstraintWeights_2:float=0.25
   ) -> (Pose) {
    alphaInputType=Bool
    boneToModify {boneName=dyn_bang_l_04}
    constraintSetup [{targetBone {boneName=spine_05}}, {targetBone {boneName=clavicle_l}}, {targetBone {boneName=spine_04}}]
    constraintWeights [1,1,1]
   }
   n36 = RigidBody!(ComponentPose=@n35, Alpha, OverrideWorldGravity:Vector=Get GravityOverride(), ExternalForce) -> (Pose) {
    bAllowPoseScalePropagation=false
    bClampLinearTranslationLimitToRefPose=true
    bOverrideWorldGravity=true
    overridePhysicsAsset=&/F_MED_ForestJeanGrowl_FaceAcc_Physics
    simulationSpace=BaseBoneSpace
    baseBoneRef {boneName=spine_05}
    componentAppliedLinearAccClamp=<3000 3000 3000>
    overrideWorldGravity {z=-980}
    simSpaceSettings {
     dampingAlpha=0
     maxAngularAcceleration=32
     maxAngularVelocity=7
     maxLinearAcceleration=2700
     maxLinearVelocity=700
     velocityScaleZ=0.3
     worldAlpha=1
     externalLinearDragV=<0.1 0.1 0.1>
    }
   }
   n37 = Anim Dynamics - Bone: dyn_ponytail_flyaway_top(ComponentPose=@n36, Alpha=1) {
    angularDampingOverride=1
    angularSpringConstant=23000
    bAngularSpring=true
    bOverrideAngularDamping=true
    bUseGravityOverride=true
    gravityScale=0.35
    simulationSpace=RootRelative
    alphaBoolBlend {blendOption=Linear}
    boundBone {boneName=dyn_ponytail_flyaway_top}
    componentAppliedLinearAccClamp=<2000 2000 2000>
    componentLinearAccScale=<1 1 1>
    componentLinearVelScale=<1 1 1>
    gravityOverride {z=-980}
    physicsBodyDefinitions [{
      boundBone {boneName=dyn_ponytail_flyaway_top}
      boxExtents=<30 30 30>
      constraintSetup {angularLimitsMax=<15 15 15>, angularLimitsMin=<-15 -15 -15>, angularTarget=<1 0 0>}
      localJointOffset=<3 0 0>
     }]
   }
   n38 = Anim Dynamics - Bone: dyn_bang_mid_front!(ComponentPose=@n37, Alpha=1, ExternalForce="0,0,0") -> (Pose) {
    angularDampingOverride=1
    bOverrideAngularDamping=true
    bUseGravityOverride=true
    simulationSpace=RootRelative
    alphaBoolBlend {blendOption=Linear}
    boundBone {boneName=dyn_bang_mid_front}
    componentLinearAccScale=<1 1 1>
    componentLinearVelScale=<1 1 1>
    gravityOverride {z=-980}
    physicsBodyDefinitions [{
      boundBone {boneName=dyn_bang_mid_front}
      boxExtents=<12 12 12>
      constraintSetup {angularLimitsMax=<1 20 2>, angularLimitsMin=<-1 -20 -25>, angularTarget=<0 0 0>}
      localJointOffset=<1 0 0>
     }]
    simSpaceSettings {maxAngularAcceleration=35, maxAngularVelocity=7, simSpaceAngularAlpha=1}
   }
   n39 = Anim Dynamics - Bone: dyn_bang_l_front(ComponentPose=@n38, Alpha=1) {
    angularDampingOverride=1
    angularSpringConstant=2000
    bAngularSpring=true
    bOverrideAngularDamping=true
    bUseGravityOverride=true
    gravityScale=0.35
    simulationSpace=RootRelative
    alphaBoolBlend {blendOption=Linear}
    boundBone {boneName=dyn_bang_l_front}
    componentAppliedLinearAccClamp=<2000 2000 2000>
    componentLinearAccScale=<1 1 1>
    componentLinearVelScale=<1 1 1>
    gravityOverride {z=-980}
    physicsBodyDefinitions [{
      boundBone {boneName=dyn_bang_l_front}
      boxExtents=<30 30 30>
      constraintSetup {angularLimitsMax=<15 5 5>, angularLimitsMin=<-15 -45 -45>, angularTarget=<1 0 0>}
      localJointOffset=<3 0 0>
     }]
   }
   n40 = Anim Dynamics - Bone: dyn_bang_r_front(ComponentPose=@n39, Alpha=1) {
    angularDampingOverride=1
    angularSpringConstant=2000
    bAngularSpring=true
    bOverrideAngularDamping=true
    bUseGravityOverride=true
    gravityScale=0.35
    simulationSpace=RootRelative
    alphaBoolBlend {blendOption=Linear}
    boundBone {boneName=dyn_bang_r_front}
    componentAppliedLinearAccClamp=<2000 2000 2000>
    componentLinearAccScale=<1 1 1>
    componentLinearVelScale=<1 1 1>
    gravityOverride {z=-980}
    physicsBodyDefinitions [{
      boundBone {boneName=dyn_bang_r_front}
      boxExtents=<30 30 30>
      constraintSetup {angularLimitsMax=<15 30 0>, angularLimitsMin=<-5 -5 -15>, angularTarget=<1 0 0>}
      localJointOffset=<3 0 0>
     }]
   }
   n41 = Anim Dynamics - Bone: dyn_ponytail_flyaway_upperL(ComponentPose=@n40, Alpha=1) {
    angularDampingOverride=1
    angularSpringConstant=2000
    bAngularSpring=true
    bOverrideAngularDamping=true
    bUseGravityOverride=true
    gravityScale=0.35
    simulationSpace=RootRelative
    alphaBoolBlend {blendOption=Linear}
    boundBone {boneName=dyn_ponytail_flyaway_upperL}
    componentAppliedLinearAccClamp=<2000 2000 2000>
    componentLinearAccScale=<1 1 1>
    componentLinearVelScale=<1 1 1>
    gravityOverride {z=-980}
    physicsBodyDefinitions [{
      boundBone {boneName=dyn_ponytail_flyaway_upperL}
      boxExtents=<30 30 30>
      constraintSetup {angularLimitsMax=<15 15 15>, angularLimitsMin=<-15 -15 -15>, angularTarget=<1 0 0>}
      localJointOffset=<3 0 0>
     }]
   }
   n42 = Anim Dynamics - Bone: dyn_ponytail_flyaway_upperR(ComponentPose=@n41, Alpha=1) {
    angularDampingOverride=1
    angularSpringConstant=2000
    bAngularSpring=true
    bOverrideAngularDamping=true
    bUseGravityOverride=true
    gravityScale=0.35
    simulationSpace=RootRelative
    alphaBoolBlend {blendOption=Linear}
    boundBone {boneName=dyn_ponytail_flyaway_upperR}
    componentAppliedLinearAccClamp=<2000 2000 2000>
    componentLinearAccScale=<1 1 1>
    componentLinearVelScale=<1 1 1>
    gravityOverride {z=-980}
    physicsBodyDefinitions [{
      boundBone {boneName=dyn_ponytail_flyaway_upperR}
      boxExtents=<30 30 30>
      constraintSetup {angularLimitsMax=<15 15 15>, angularLimitsMin=<-15 -15 -15>, angularTarget=<1 0 0>}
      localJointOffset=<3 0 0>
     }]
   }
   n43 = Anim Dynamics - Bone: dyn_ponytail_flyaway_mid_l(ComponentPose=@n42, Alpha=1) {
    angularDampingOverride=1
    angularSpringConstant=2000
    bAngularSpring=true
    bOverrideAngularDamping=true
    bUseGravityOverride=true
    gravityScale=0.35
    simulationSpace=RootRelative
    alphaBoolBlend {blendOption=Linear}
    boundBone {boneName=dyn_ponytail_flyaway_mid_l}
    componentAppliedLinearAccClamp=<2000 2000 2000>
    componentLinearAccScale=<1 1 1>
    componentLinearVelScale=<1 1 1>
    gravityOverride {z=-980}
    physicsBodyDefinitions [{
      boundBone {boneName=dyn_ponytail_flyaway_mid_l}
      boxExtents=<30 30 30>
      constraintSetup {angularLimitsMax=<15 15 15>, angularLimitsMin=<-15 -15 -15>, angularTarget=<1 0 0>}
      localJointOffset=<3 0 0>
     }]
   }
   n44 = Anim Dynamics - Bone: dyn_ponytail_flyaway_mid_r(ComponentPose=@n43, Alpha=1) {
    angularDampingOverride=1
    angularSpringConstant=2000
    bAngularSpring=true
    bOverrideAngularDamping=true
    bUseGravityOverride=true
    gravityScale=0.35
    simulationSpace=RootRelative
    alphaBoolBlend {blendOption=Linear}
    boundBone {boneName=dyn_ponytail_flyaway_mid_r}
    componentAppliedLinearAccClamp=<2000 2000 2000>
    componentLinearAccScale=<1 1 1>
    componentLinearVelScale=<1 1 1>
    gravityOverride {z=-980}
    physicsBodyDefinitions [{
      boundBone {boneName=dyn_ponytail_flyaway_mid_r}
      boxExtents=<30 30 30>
      constraintSetup {angularLimitsMax=<15 15 15>, angularLimitsMin=<-15 -15 -15>, angularTarget=<1 0 0>}
      localJointOffset=<3 0 0>
     }]
   }
   n45 = "Transform (Modify) Bone - Bone: dyn_ponytail_flyaway_mid_r"(ComponentPose=@n44, Alpha=1, Translation="0,-1,1", Rotation="5,0,10", Scale="1,1,1") {
    rotationMode=BMM_Additive
    translationMode=BMM_Additive
    boneToModify {boneName=dyn_ponytail_flyaway_mid_r}
    rotation {pitch=5, roll=10}
    translation {y=-1, z=1}
   }
   n46 = "Transform (Modify) Bone - Bone: dyn_ponytail_mstr"!(ComponentPose=@n45, bAlphaBoolEnabled:bool=Get Is Backbling A Cape(), Translation="0,0,0", Rotation="0,-17,0", Scale="1,1,1") -> (Pose) {
    alphaInputType=Bool
    rotationMode=BMM_Additive
    rotationSpace=BCS_BoneSpace
    translationMode=BMM_Additive
    boneToModify {boneName=dyn_ponytail_mstr}
    rotation {yaw=-17}
   }
   Save cached pose 'RBanPose'(Pose=Component To Local(ComponentPose=@n46)) {cacheName=RBanPose}
   n47 = "Constraint - dyn_ponytail_01 (P)"!(ComponentPose=Local To Component(LocalPose=Use cached pose 'InitialPose'()), bAlphaBoolEnabled:bool=True, ConstraintWeights_0=1) -> (Pose) {alphaInputType=Bool, boneToModify {boneName=dyn_ponytail_01}, constraintSetup [{targetBone {boneName=head}}], constraintWeights [1]}
   n48 = "Constraint - dyn_ponytail_02 (PP)"!(ComponentPose=@n47, bAlphaBoolEnabled:bool=True, ConstraintWeights_0=0.8, ConstraintWeights_1:float=0.2) -> (Pose) {
    alphaInputType=Bool
    boneToModify {boneName=dyn_ponytail_02}
    constraintSetup [{targetBone {boneName=head}}, {targetBone {boneName=neck_01}}]
    constraintWeights [1,1]
   }
   n49 = "Constraint - dyn_ponytail_03 (PPP)"!(ComponentPose=@n48, bAlphaBoolEnabled:bool=True, ConstraintWeights_0=1, ConstraintWeights_1=0, ConstraintWeights_2=0) -> (Pose) {
    alphaInputType=Bool
    boneToModify {boneName=dyn_ponytail_03}
    constraintSetup [{targetBone {boneName=neck_01}}, {targetBone {boneName=spine_05}}, {targetBone {boneName=head}}]
    constraintWeights [1,1,1]
   }
   n50 = "Constraint - dyn_ponytail_04 (PP)"!(ComponentPose=@n49, bAlphaBoolEnabled:bool=True, ConstraintWeights_0=0.7, ConstraintWeights_1=0.3) -> (Pose) {
    alphaInputType=Bool
    boneToModify {boneName=dyn_ponytail_04}
    constraintSetup [{targetBone {boneName=spine_04}}, {targetBone {boneName=neck_01}}]
    constraintWeights [1,1]
   }
   n51 = "Constraint - dyn_ponytail_05 (PPP)"!(ComponentPose=@n50, bAlphaBoolEnabled:bool=True, ConstraintWeights_0=0.7, ConstraintWeights_1:float=0.2, ConstraintWeights_2:float=0.1) -> (Pose) {
    alphaInputType=Bool
    boneToModify {boneName=dyn_ponytail_05}
    constraintSetup [{targetBone {boneName=spine_03}}, {targetBone {boneName=neck_01}}, {targetBone {boneName=pelvis}}]
    constraintWeights [1,1,1]
   }
   n52 = "Constraint - dyn_ponytail_flyaway_mid_r (PP)"!(ComponentPose=@n51, bAlphaBoolEnabled:bool=True, ConstraintWeights_0=0.4, ConstraintWeights_1:float=0.6) -> (Pose) {
    alphaInputType=Bool
    boneToModify {boneName=dyn_ponytail_flyaway_mid_r}
    constraintSetup [{targetBone {boneName=dyn_ponytail_02}}, {targetBone {boneName=dyn_ponytail_03}}]
    constraintWeights [1,1]
   }
   n53 = "Constraint - dyn_ponytail_flyaway_mid_l (PP)"!(ComponentPose=@n52, bAlphaBoolEnabled:bool=True, ConstraintWeights_0=0.4, ConstraintWeights_1:float=0.6) -> (Pose) {
    alphaInputType=Bool
    boneToModify {boneName=dyn_ponytail_flyaway_mid_l}
    constraintSetup [{targetBone {boneName=dyn_ponytail_02}}, {targetBone {boneName=dyn_ponytail_03}}]
    constraintWeights [1,1]
   }
   n54 = "Transform (Modify) Bone - Bone: dyn_ponytail_03"(ComponentPose=@n53, bAlphaBoolEnabled=True, Translation="0,0,0", Rotation="0,0,-5", Scale="1,1,1") {
    alphaInputType=Bool
    rotationMode=BMM_Additive
    scaleMode=BMM_Additive
    translationMode=BMM_Additive
    boneToModify {boneName=dyn_ponytail_03}
    rotation {roll=-5}
   }
   n55 = "Transform (Modify) Bone - Bone: dyn_ponytail_04"!(ComponentPose=@n54, bAlphaBoolEnabled:bool=True, Translation="0,0,3", Rotation="0.000,0.000,-10.000", Scale="1,1,1") -> (Pose) {
    alphaInputType=Bool
    rotationMode=BMM_Additive
    scaleMode=BMM_Additive
    translationMode=BMM_Additive
    boneToModify {boneName=dyn_ponytail_04}
    rotation {roll=-10}
    translation {z=3}
   }
   n56 = "Transform (Modify) Bone - Bone: dyn_ponytail_05"!(ComponentPose=@n55, bAlphaBoolEnabled:bool=True, Translation="0,0,0", Rotation="0,0,20", Scale="1,1,1") -> (Pose) {
    alphaInputType=Bool
    rotationMode=BMM_Additive
    scaleMode=BMM_Additive
    translationMode=BMM_Additive
    boneToModify {boneName=dyn_ponytail_05}
    rotation {roll=20}
   }
   n57 = Component To Local(ComponentPose=@n56)
   n58 = "Transform (Modify) Bone - Bone: dyn_ponytail_05"(
    ComponentPose=Local To Component(LocalPose=Use cached pose 'InitialPose'())
    bAlphaBoolEnabled=True
    Translation="0,1,0"
    Rotation="0,0,0"
    Scale="1,1,1"
   ) {
    alphaInputType=Bool
    rotationMode=BMM_Additive
    rotationSpace=BCS_BoneSpace
    translationMode=BMM_Additive
    translationSpace=BCS_BoneSpace
    alphaBoolBlend {blendOption=Linear}
    boneToModify {boneName=dyn_ponytail_05}
    translation {y=1}
   }
   n59 = Component To Local(ComponentPose=@n58)
   n60 = Get ForestJeanGrowl_Backpack() -> (ForestJeanGrowl_Backpack:bool)
   n61 = Anim Dynamics - Bone: dyn_ponytail_flyaway_top(
    ComponentPose=Local To Component(LocalPose=Blend Poses by bool(BlendPose_0=@n59, BlendPose_1=@n57, BlendTime_0=0, BlendTime_1=0, bActiveValue=@n60) {blendType=Linear})
    Alpha=1
   ) {
    angularDampingOverride=1
    angularSpringConstant=23000
    bAngularSpring=true
    bOverrideAngularDamping=true
    bUseGravityOverride=true
    gravityScale=0.35
    simulationSpace=RootRelative
    alphaBoolBlend {blendOption=Linear}
    boundBone {boneName=dyn_ponytail_flyaway_top}
    componentAppliedLinearAccClamp=<2000 2000 2000>
    componentLinearAccScale=<1 1 1>
    componentLinearVelScale=<1 1 1>
    gravityOverride {z=-980}
    physicsBodyDefinitions [{
      boundBone {boneName=dyn_ponytail_flyaway_top}
      boxExtents=<30 30 30>
      constraintSetup {angularLimitsMax=<15 15 15>, angularLimitsMin=<-15 -15 -15>, angularTarget=<1 0 0>}
      localJointOffset=<3 0 0>
     }]
   }
   n62 = Anim Dynamics - Bone: dyn_ponytail_01!(ComponentPose=@n61, Alpha=1, ExternalForce="0,0,0") -> (Pose) {
    angularDampingOverride=0.98
    angularSpringConstant=400
    bAngularSpring=true
    bOverrideAngularDamping=true
    bUseGravityOverride=true
    gravityScale=0.35
    simulationSpace=RootRelative
    alphaBoolBlend {blendOption=Linear}
    boundBone {boneName=dyn_ponytail_01}
    componentAppliedLinearAccClamp=<1300 1300 1300>
    componentLinearAccScale=<1 1 1>
    componentLinearVelScale=<1 1 1>
    gravityOverride {z=-980}
    physicsBodyDefinitions [{
      boundBone {boneName=dyn_ponytail_01}
      boxExtents=<30 40 40>
      constraintSetup {angularLimitsMax=<5 5 0>, angularLimitsMin=<-5 -10 -1>, angularTarget=<1 0 0>}
      localJointOffset=<4 0 0>
     }]
   }
   n63 = Anim Dynamics - Bone: dyn_ponytail_02!(ComponentPose=@n62, Alpha=1, ExternalForce="0,0,0") -> (Pose) {
    angularDampingOverride=0.98
    angularSpringConstant=400
    bAngularSpring=true
    bOverrideAngularDamping=true
    bUseGravityOverride=true
    gravityScale=0.35
    simulationSpace=RootRelative
    alphaBoolBlend {blendOption=Linear}
    boundBone {boneName=dyn_ponytail_02}
    componentAppliedLinearAccClamp=<1300 1300 1300>
    componentLinearAccScale=<1 1 1>
    componentLinearVelScale=<1 1 1>
    gravityOverride {z=-980}
    physicsBodyDefinitions [{
      boundBone {boneName=dyn_ponytail_02}
      boxExtents=<30 40 40>
      constraintSetup {angularLimitsMax=<5 5 0>, angularLimitsMin=<-5 -10 -1>, angularTarget=<1 0 0>}
      localJointOffset=<4 0 0>
     }]
   }
   n64 = Anim Dynamics - Bone: dyn_ponytail_03!(ComponentPose=@n63, Alpha=1, ExternalForce="0,0,0") -> (Pose) {
    angularDampingOverride=0.98
    angularSpringConstant=400
    bAngularSpring=true
    bOverrideAngularDamping=true
    bUseGravityOverride=true
    gravityScale=0.35
    simulationSpace=RootRelative
    alphaBoolBlend {blendOption=Linear}
    boundBone {boneName=dyn_ponytail_03}
    componentAppliedLinearAccClamp=<1300 1300 1600>
    componentLinearAccScale=<1 1 1>
    componentLinearVelScale=<1 1 1>
    gravityOverride {z=-980}
    physicsBodyDefinitions [{
      boundBone {boneName=dyn_ponytail_03}
      boxExtents=<30 40 40>
      constraintSetup {angularLimitsMax=<5 5 0>, angularLimitsMin=<-5 -10 -1>, angularTarget=<1 0 0>}
      localJointOffset=<4 0 0>
     }]
   }
   n65 = Anim Dynamics - Bone: dyn_ponytail_04!(ComponentPose=@n64, Alpha=1, ExternalForce="0,0,0") -> (Pose) {
    angularDampingOverride=0.98
    angularSpringConstant=400
    bAngularSpring=true
    bOverrideAngularDamping=true
    bUseGravityOverride=true
    gravityScale=0.35
    simulationSpace=RootRelative
    alphaBoolBlend {blendOption=Linear}
    boundBone {boneName=dyn_ponytail_04}
    componentAppliedLinearAccClamp=<1300 1300 2400>
    componentLinearAccScale=<1 1 1>
    componentLinearVelScale=<1 1 1>
    gravityOverride {z=-980}
    physicsBodyDefinitions [{
      boundBone {boneName=dyn_ponytail_04}
      boxExtents=<30 40 40>
      constraintSetup {angularLimitsMax=<5 5 0>, angularLimitsMin=<-5 -5 -5>, angularTarget=<1 0 0>}
      localJointOffset=<4 0 0>
     }]
   }
   n66 = Anim Dynamics - Bone: dyn_ponytail_05!(ComponentPose=@n65, Alpha=1, ExternalForce="0,0,0") -> (Pose) {
    angularDampingOverride=0.98
    angularSpringConstant=400
    bAngularSpring=true
    bOverrideAngularDamping=true
    bUseGravityOverride=true
    gravityScale=0.35
    simulationSpace=RootRelative
    alphaBoolBlend {blendOption=Linear}
    boundBone {boneName=dyn_ponytail_05}
    componentAppliedLinearAccClamp=<1300 1300 2400>
    componentLinearAccScale=<1 1 1>
    componentLinearVelScale=<1 1 1>
    gravityOverride {z=-980}
    physicsBodyDefinitions [{
      boundBone {boneName=dyn_ponytail_05}
      boxExtents=<30 40 40>
      constraintSetup {angularLimitsMax=<5 5 0>, angularLimitsMin=<-5 -5 -5>, angularTarget=<1 0 0>}
      localJointOffset=<4 0 0>
     }]
   }
   n67 = Anim Dynamics - Bone: dyn_ponytail_flyaway_upperL(ComponentPose=@n66, Alpha=1) {
    angularDampingOverride=1
    angularSpringConstant=2000
    bAngularSpring=true
    bOverrideAngularDamping=true
    bUseGravityOverride=true
    gravityScale=0.35
    simulationSpace=RootRelative
    alphaBoolBlend {blendOption=Linear}
    boundBone {boneName=dyn_ponytail_flyaway_upperL}
    componentAppliedLinearAccClamp=<2000 2000 2000>
    componentLinearAccScale=<1 1 1>
    componentLinearVelScale=<1 1 1>
    gravityOverride {z=-980}
    physicsBodyDefinitions [{
      boundBone {boneName=dyn_ponytail_flyaway_upperL}
      boxExtents=<30 30 30>
      constraintSetup {angularLimitsMax=<15 15 15>, angularLimitsMin=<-15 -15 -15>, angularTarget=<1 0 0>}
      localJointOffset=<3 0 0>
     }]
   }
   n68 = Anim Dynamics - Bone: dyn_ponytail_flyaway_upperR(ComponentPose=@n67, Alpha=1) {
    angularDampingOverride=1
    angularSpringConstant=2000
    bAngularSpring=true
    bOverrideAngularDamping=true
    bUseGravityOverride=true
    gravityScale=0.35
    simulationSpace=RootRelative
    alphaBoolBlend {blendOption=Linear}
    boundBone {boneName=dyn_ponytail_flyaway_upperR}
    componentAppliedLinearAccClamp=<2000 2000 2000>
    componentLinearAccScale=<1 1 1>
    componentLinearVelScale=<1 1 1>
    gravityOverride {z=-980}
    physicsBodyDefinitions [{
      boundBone {boneName=dyn_ponytail_flyaway_upperR}
      boxExtents=<30 30 30>
      constraintSetup {angularLimitsMax=<15 15 15>, angularLimitsMin=<-15 -15 -15>, angularTarget=<1 0 0>}
      localJointOffset=<3 0 0>
     }]
   }
   n69 = Anim Dynamics - Bone: dyn_ponytail_flyaway_mid_l(ComponentPose=@n68, Alpha=1) {
    angularDampingOverride=1
    angularSpringConstant=2000
    bAngularSpring=true
    bOverrideAngularDamping=true
    bUseGravityOverride=true
    gravityScale=0.35
    simulationSpace=RootRelative
    alphaBoolBlend {blendOption=Linear}
    boundBone {boneName=dyn_ponytail_flyaway_mid_l}
    componentAppliedLinearAccClamp=<2000 2000 2000>
    componentLinearAccScale=<1 1 1>
    componentLinearVelScale=<1 1 1>
    gravityOverride {z=-980}
    physicsBodyDefinitions [{
      boundBone {boneName=dyn_ponytail_flyaway_mid_l}
      boxExtents=<30 30 30>
      constraintSetup {angularLimitsMax=<15 15 15>, angularLimitsMin=<-15 -15 -15>, angularTarget=<1 0 0>}
      localJointOffset=<3 0 0>
     }]
   }
   n70 = Anim Dynamics - Bone: dyn_ponytail_flyaway_mid_r(ComponentPose=@n69, Alpha=1) {
    angularDampingOverride=1
    angularSpringConstant=2000
    bAngularSpring=true
    bOverrideAngularDamping=true
    bUseGravityOverride=true
    gravityScale=0.35
    simulationSpace=RootRelative
    alphaBoolBlend {blendOption=Linear}
    boundBone {boneName=dyn_ponytail_flyaway_mid_r}
    componentAppliedLinearAccClamp=<2000 2000 2000>
    componentLinearAccScale=<1 1 1>
    componentLinearVelScale=<1 1 1>
    gravityOverride {z=-980}
    physicsBodyDefinitions [{
      boundBone {boneName=dyn_ponytail_flyaway_mid_r}
      boxExtents=<30 30 30>
      constraintSetup {angularLimitsMax=<15 15 15>, angularLimitsMin=<-15 -15 -15>, angularTarget=<1 0 0>}
      localJointOffset=<3 0 0>
     }]
   }
   n71 = NOT Boolean(A=Get bRigidBodyEnabled())
   n72 = Anim Dynamics - Bone: dyn_bang_mid_front!(SimulationSpace:AnimPhysSimSpaceType Enum, ComponentPose=@n70, bAlphaBoolEnabled:bool=@n71, ExternalForce="0,0,0") -> (Pose) {
    alphaInputType=Bool
    angularDampingOverride=1
    bOverrideAngularDamping=true
    bUseGravityOverride=true
    simulationSpace=RootRelative
    alphaBoolBlend {blendOption=Linear}
    boundBone {boneName=dyn_bang_mid_front}
    componentAppliedLinearAccClamp=<2000 2000 2000>
    componentLinearAccScale=<1 1 1>
    componentLinearVelScale=<1 1 1>
    gravityOverride {z=-980}
    physicsBodyDefinitions [{
      boundBone {boneName=dyn_bang_mid_front}
      boxExtents=<12 12 12>
      constraintSetup {angularLimitsMax=<1 20 2>, angularLimitsMin=<-1 -20 -25>, angularTarget=<0 0 0>}
      localJointOffset=<1 0 0>
     }]
    simSpaceSettings {maxAngularAcceleration=35, maxAngularVelocity=7, simSpaceAngularAlpha=1}
   }
   n73 = Anim Dynamics - Bone: dyn_bang_l_front!(SimulationSpace, ComponentPose=@n72, bAlphaBoolEnabled:bool=@n71) -> (Pose) {
    alphaInputType=Bool
    angularDampingOverride=1
    angularSpringConstant=2000
    bAngularSpring=true
    bOverrideAngularDamping=true
    bUseGravityOverride=true
    gravityScale=0.35
    simulationSpace=RootRelative
    alphaBoolBlend {blendOption=Linear}
    boundBone {boneName=dyn_bang_l_front}
    componentAppliedLinearAccClamp=<2000 2000 2000>
    componentLinearAccScale=<1 1 1>
    componentLinearVelScale=<1 1 1>
    gravityOverride {z=-980}
    physicsBodyDefinitions [{
      boundBone {boneName=dyn_bang_l_front}
      boxExtents=<30 30 30>
      constraintSetup {angularLimitsMax=<15 5 5>, angularLimitsMin=<-15 -45 -45>, angularTarget=<1 0 0>}
      localJointOffset=<3 0 0>
     }]
   }
   n74 = Anim Dynamics - Bone: dyn_bang_r_front!(SimulationSpace, ComponentPose=@n73, bAlphaBoolEnabled:bool=@n71) -> (Pose) {
    alphaInputType=Bool
    angularDampingOverride=1
    angularSpringConstant=2000
    bAngularSpring=true
    bOverrideAngularDamping=true
    bUseGravityOverride=true
    gravityScale=0.35
    simulationSpace=RootRelative
    alphaBoolBlend {blendOption=Linear}
    boundBone {boneName=dyn_bang_r_front}
    componentAppliedLinearAccClamp=<2000 2000 2000>
    componentLinearAccScale=<1 1 1>
    componentLinearVelScale=<1 1 1>
    gravityOverride {z=-980}
    physicsBodyDefinitions [{
      boundBone {boneName=dyn_bang_r_front}
      boxExtents=<30 30 30>
      constraintSetup {angularLimitsMax=<15 30 0>, angularLimitsMin=<-5 -5 -15>, angularTarget=<1 0 0>}
      localJointOffset=<3 0 0>
     }]
   }
   n75 = Anim Dynamics - Bone: dyn_bang_l_02!(SimulationSpace, ComponentPose=@n74, bAlphaBoolEnabled:bool=@n71) -> (Pose) {
    alphaInputType=Bool
    angularDampingOverride=1
    angularSpringConstant=2000
    bAngularSpring=true
    bOverrideAngularDamping=true
    bUseGravityOverride=true
    gravityScale=0.35
    simulationSpace=RootRelative
    alphaBoolBlend {blendOption=Linear}
    boundBone {boneName=dyn_bang_l_02}
    componentAppliedLinearAccClamp=<2000 2000 2000>
    componentLinearAccScale=<1 1 1>
    componentLinearVelScale=<1 1 1>
    gravityOverride {z=-980}
    physicsBodyDefinitions [{
      boundBone {boneName=dyn_bang_l_02}
      boxExtents=<30 30 30>
      constraintSetup {angularLimitsMax=<15 15 5>, angularLimitsMin=<-15 -45 -45>, angularTarget=<1 0 0>}
      localJointOffset=<3 0 0>
     }]
   }
   n76 = Anim Dynamics - Bone: dyn_bang_r_02!(SimulationSpace, ComponentPose=@n75, bAlphaBoolEnabled:bool=@n71) -> (Pose) {
    alphaInputType=Bool
    angularDampingOverride=1
    angularSpringConstant=2000
    bAngularSpring=true
    bOverrideAngularDamping=true
    bUseGravityOverride=true
    gravityScale=0.35
    simulationSpace=RootRelative
    alphaBoolBlend {blendOption=Linear}
    boundBone {boneName=dyn_bang_r_02}
    componentAppliedLinearAccClamp=<2000 2000 2000>
    componentLinearAccScale=<1 1 1>
    componentLinearVelScale=<1 1 1>
    gravityOverride {z=-980}
    physicsBodyDefinitions [{
      boundBone {boneName=dyn_bang_r_02}
      boxExtents=<30 30 30>
      constraintSetup {angularLimitsMax=<15 30 0>, angularLimitsMin=<-5 -15 -15>, angularTarget=<1 0 0>}
      localJointOffset=<3 0 0>
     }]
   }
   n77 = Anim Dynamics - Bone: dyn_bang_l_03!(SimulationSpace, ComponentPose=@n76, bAlphaBoolEnabled:bool=@n71) -> (Pose) {
    alphaInputType=Bool
    angularDampingOverride=1
    angularSpringConstant=2000
    bAngularSpring=true
    bOverrideAngularDamping=true
    bUseGravityOverride=true
    gravityScale=0.35
    simulationSpace=RootRelative
    alphaBoolBlend {blendOption=Linear}
    boundBone {boneName=dyn_bang_l_03}
    componentAppliedLinearAccClamp=<2000 2000 2000>
    componentLinearAccScale=<1 1 1>
    componentLinearVelScale=<1 1 1>
    gravityOverride {z=-980}
    physicsBodyDefinitions [{
      boundBone {boneName=dyn_bang_l_03}
      boxExtents=<30 30 30>
      constraintSetup {angularLimitsMax=<15 15 5>, angularLimitsMin=<-15 -45 -45>, angularTarget=<1 0 0>}
      localJointOffset=<3 0 0>
     }]
   }
   n78 = Anim Dynamics - Bone: dyn_bang_r_03!(SimulationSpace, ComponentPose=@n77, bAlphaBoolEnabled:bool=@n71) -> (Pose) {
    alphaInputType=Bool
    angularDampingOverride=1
    angularSpringConstant=2000
    bAngularSpring=true
    bOverrideAngularDamping=true
    bUseGravityOverride=true
    gravityScale=0.35
    simulationSpace=RootRelative
    alphaBoolBlend {blendOption=Linear}
    boundBone {boneName=dyn_bang_r_03}
    componentAppliedLinearAccClamp=<2000 2000 2000>
    componentLinearAccScale=<1 1 1>
    componentLinearVelScale=<1 1 1>
    gravityOverride {z=-980}
    physicsBodyDefinitions [{
      boundBone {boneName=dyn_bang_r_03}
      boxExtents=<30 30 30>
      constraintSetup {angularLimitsMax=<15 30 0>, angularLimitsMin=<-5 -15 -15>, angularTarget=<1 0 0>}
      localJointOffset=<3 0 0>
     }]
   }
   n79 = RigidBody!(ComponentPose=@n78, Alpha, ExternalForce) -> (Pose) {
    bClampLinearTranslationLimitToRefPose=true
    bOverrideWorldGravity=true
    overridePhysicsAsset=&/F_MED_ForestJeanGrowl_FaceAcc_Front_Physics
    simulationSpace=BaseBoneSpace
    baseBoneRef {boneName=neck_01}
    componentAppliedLinearAccClamp=<3000 3000 3000>
    overrideWorldGravity {z=-980}
    simSpaceSettings {
     dampingAlpha=0
     maxAngularAcceleration=35
     maxAngularVelocity=7
     maxLinearAcceleration=3000
     maxLinearVelocity=700
     velocityScaleZ=0
     worldAlpha=0.5
     externalLinearDragV=<1 1 1>
    }
   }
   Save cached pose 'BackpackPose'(Pose=Component To Local(ComponentPose=@n79)) {cacheName=BackpackPose}
  ]}
 EventGraph {nodes [
   Event BlueprintUpdateAnimation()
   > n1 = Sequence!(execute) -> (then_0, then_1, then_2:Exec, then_3:Exec)
   Set frontendAlpha(execute=@n1.then_1, frontendAlpha:double=SelectFloat(A=0.5, B=1, bPickA=Get bIsInFrontEnd())) -> (Output_Get:double)
   > n2 = Branch(Condition=OR Boolean(A=Get bIsInFrontEnd(), B=Get bIsDBNOCarried()))
   > Set simSpaceChange(simSpaceChange:AnimPhysSimSpaceType Enum=World) -> (Output_Get:AnimPhysSimSpaceType Enum)
   Set simSpaceChange(execute=@n2.else)
   n3 = NOT Boolean(A=Get Is Wearing Backbling())
   n4 = Get bIsSkydiveFloating()
   Set Skydive(execute=@n1.then_3, Skydive:double=SelectFloat(A=0.75, B=0, bPickA=AND Boolean(A=@n4, B=@n3))) -> (Output_Get:double)
   > Set Moving(Moving:double=SelectFloat(A=1, B=0, bPickA=Get bIsMoving2D())) -> (Output_Get:double)
   n5 = "Gravity Override State Function (Parameter Asset) (Nativized)"(
    execute=@n1.then_0
    CharacterPartAnimInstance=Self-Reference()
    GameTimeInSeconds=GetGameTimeInSeconds()
    Parameters=&/F_MED_ForestJeanGrowlFA_Parameters
   )
   > Set GravityOverride(GravityOverride=@n5)
   n6 = Event OnCharacterCustomizationLoaded()
   > n7 = Sequence!(execute) -> (then_0, then_1)
   n8 = Branch(
    execute=@n7.then_1
    Condition=CharacterItemDefinitionHasAnyMetaTag(Tags=Make Array("[0]"=(TagName="Cosmetics.MetaTags.ForestJeanGrowl.Joy")))
   )
   > Set ForestJeanGrowl_Joy(ForestJeanGrowl_Joy:bool=true) -> (Output_Get:bool)
   Set ForestJeanGrowl_Joy(execute=@n8.else, ForestJeanGrowl_Joy=false)
   n9 = CheckForBackbling()
   Set Is Wearing Backbling(execute=@n7.then_0, Is Wearing Backbling=AND Boolean(A=@n9, B=NOT Boolean(A=@n9.bIsBackblingACape)))
   > Set Is Backbling A Cape(Is Backbling A Cape:bool=@n9.bIsBackblingACape) -> (Output_Get:bool)
   > Set Backbling(Backbling:Skeletal Mesh Component Budgeted Object Reference=Get Array Item with EFort Custom Part Type(Array=@n6.CustomMeshParts, PartType=Backpack)) -> (Output_Get:Skeletal Mesh Component Budgeted Object Reference)
   > n10 = Branch(Condition=BackpackItemDefinitionHasAnyMetaTag!(Tags:Array of Gameplay Tag Structures=Make Array("[0]"=(TagName="Cosmetics.MetaTags.ForestJeanGrowl")), bRequiresEquipped:bool=false) -> (ReturnValue:bool))
   > Set ForestJeanGrowl_Backpack(ForestJeanGrowl_Backpack:bool=true) -> (Output_Get:bool)
   Set ForestJeanGrowl_Backpack(execute=@n10.else, ForestJeanGrowl_Backpack=false)
   Event OnBeginCrouching()
   > Set Crouch(Crouch:double=0.7) -> (Output_Get:double)
   Event OnEndCrouching()
   > Set Crouch(Crouch=0)
  ]}
]
variables [
 AccScaleOverride {type=Vector}
 Backbling {type=SkeletalMeshComponentBudgeted}
 Body {type=SkeletalMeshComponentBudgeted}
 Crouch {type=number}
 ForestJeanGrowl_Backpack {type=boolean}
 ForestJeanGrowl_Joy {type=boolean}
 frontendAlpha {type=number}
 GravityOverride {type=Vector}
 Is Backbling A Cape {type=boolean}
 Is Wearing Backbling {type=boolean}
 Moving {type=number}
 simSpaceChange {type=AnimPhysSimSpaceType}
 Skydive {type=number}
 VelScaleOverride {type=Vector}
]
